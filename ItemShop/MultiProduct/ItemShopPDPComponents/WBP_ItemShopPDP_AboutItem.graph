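# Bottom bar button for the Item Shop product detail page that opens an "About" dialog for the displayed item's category.
# UpdateAboutItemButton stores the item view model, shows the button only if its category supplies a dialog class, and labels it by category, with a LEGO Kits override for Juno building items.
# Clicking async-creates that soft-referenced widget, passes it the view model, pushes it as a modal and records the item type as seen.
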WidgetBlueprint /ItemShop/MultiProduct/ItemShopPDPComponents/WBP_ItemShopPDP_AboutItem
graphs [
 EventGraph {nodes [
   n1 = Event PreConstruct()
   > Parent: PreConstruct(IsDesignTime=@n1.IsDesignTime)
   Event Construct()
   > Parent: Construct()
   Event On Clicked()
   n2 = TryGetSoftCustomActivatableWidgetClass!(
    self:Fort Locker Category Config Object Reference=GetItemCategory!(self:Fort Item VM Object Reference=Get Item VM() -> (Item VM:Fort Item VM Object Reference)) -> (ReturnValue:Fort Locker Category Config Object Reference)
    ViewRequest:Gameplay Tag Structure=(TagName="UI.ItemShop.AboutItemDialogClass")
   ) -> (WidgetClass:Common Activatable Widget Soft Class Reference, ReturnValue:bool)
   > Branch(Condition=@n2.ReturnValue)
   > n3 = CreateWidgetAsync!(
    execute:Exec
    UserWidgetSoftClass:User Widget Soft Class Reference=@n2
    OwningPlayer:Player Controller Object Reference=GetOwningPlayer()
    bSuspendInputUntilComplete:bool=true
   ) -> (then:Exec, OnComplete:Exec, OnFailure:Exec, UserWidget:User Widget Object Reference) {is=title K2Node_AsyncAction}
   n4 = Get Item VM()
   SetViewModelByClass!(execute=@n3.OnComplete, Widget:User Widget Object Reference=@n3, ViewModel:Notify Field Value Changed Interface=@n4) -> (then)
   > TryPushWidgetAsModalPanel!(execute, PushWidget:User Widget Object Reference=@n3) -> (then)
   > SetHasSeenAboutItemModal!(
    execute
    self:Fort Game User Settings Object Reference=GetFortGameUserSettings!() -> (ReturnValue:Fort Game User Settings Object Reference)
    ItemType:EFortItemType Enum=GetItemType!(ItemDefinition:Item Definition Base Object Reference=Get ItemDefinition!(self:Fort Item VM Object Reference=@n4) -> (ItemDefinition:Fort Item Definition Object Reference)) -> (ReturnValue:EFortItemType Enum=WorldItem)
    bHasSeen:bool=true
   ) -> (then)
   Event Tick()
  ]}
 OnSetAllInteractionsEnabled {nodes [OnSetAllInteractionsEnabled!() -> (then, ShouldEnable:bool=~) {is=title K2Node_FunctionEntry}]}
 UpdateAboutItemButton {nodes [
   n1 = UpdateAboutItemButton!() -> (then, Item VM:Fort Item VM Object Reference) {is=title K2Node_FunctionEntry}
   > n2 = Set Item VM(Item VM:Fort Item VM Object Reference=@n1) -> (Output_Get:Fort Item VM Object Reference)
   > Set Category(Category:Fort Locker Category Config Object Reference=GetItemCategory(self=@n2)) -> (Output_Get:Fort Locker Category Config Object Reference)
   > SetBottomButtonVisibility!(
    execute
    self:WBP Bottom Bar Decoy Button Object Reference
    ShouldButtonBeVisible:bool=TryGetSoftCustomActivatableWidgetClass(self=Get Category() -> (Category:Fort Locker Category Config Object Reference), ViewRequest=(TagName="UI.ItemShop.AboutItemDialogClass")).ReturnValue
   ) -> (then)
   n3 = GetItemType(ItemDefinition=Get ItemDefinition(self=Get Item VM())) -> (ReturnValue=WorldItem)
   > n4 = Branch(Condition=OR Boolean(A="Equal (Enum)"(A:EFortItemType Enum=@n3, B:EFortItemType Enum=JunoBuildingSet), B="Equal (Enum)"(A=@n3, B=JunoBuildingProp)))
   > n5 = Set TargetDisplayName(TargetDisplayName:Text="LOCTEXT(\"ABOUT LEGO® KITS\")") -> (Output_Get:Text)
   > SetActionDisplayName!(execute, self:Common Button Base Object Reference, InDisplayName:"Text (by ref)"=@n5) -> (then)
   n6 = Get SlotInfo!(self:Fort Locker Category Config Object Reference=Get Category()) -> (SlotInfo:Fort Locker Category Slot Info Structure)
   SetActionDisplayName(
    execute=@n4.else
    InDisplayName=Format Text!(
     Format="LOCTEXT(\"About {ItemCategoryPluralName}\")"
     ItemCategoryPluralName:Text=Break Fort Locker Category Slot Info!(FortLockerCategorySlotInfo:"Fort Locker Category Slot Info Structure (by ref)"=@n6) -> (CategoryNamePlural:Text, CategorySoundOverrides:Map of Gameplay Tag Structures to Sound Base Soft Object References) {is=title K2Node_BreakStruct}
    ) -> (Result)
   )
  ]}
]
variables [
 AboutItemCoachmark {type=WBP_UIKit_CoachMark_C, default=null}
 Item VM {type=FortItemVM, default=null}
 OnSetAllInteractionsEnabled {}
]
widgets {
 info {
  inheritedWidgetCount=0
  parentClass=/Game/UI/Frontend/BottomBar/WBP_BottomBarDecoyButton.WBP_BottomBarDecoyButton_C
  rootWidgetClass=null
  widgetCount=0
 }
 widgets []
}
