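# Item Shop tile component that shows a cosmetic gating warning icon when an offer's cosmetics have rating restrictions for the owning local player.
# SetOfferVM stores the offer viewmodel, and PreConstruct applies the configurable IconSize to the wrapped warning icon widget.
# VM_GrantedItemsUpdated collects cosmetic definitions from granted items displayable in the current shop context.
# It converts them to gating interfaces, checks restrictions and toggles the icon between SelfHitTestInvisible and Collapsed.
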WidgetBlueprint /ItemShop/MultiProduct/ItemShopTileComponents/WBP_ItemShopTile_CosmeticGatingWarningIcon
graphs [
 EventGraph {nodes [
   n1 = Event SetOfferVM!() -> (OutputDelegate, then, OfferVM:Fort Item Shop Offer VM Object Reference) {is=title K2Node_Event}
   > SetFortItemShopOfferVM!(execute, self, Viewmodel:Fort Item Shop Offer VM Object Reference=@n1.OfferVM) -> (then)
   Event PreConstruct()
   > SetIconSize!(
    execute
    self:WBP Cosmetic Gating Warning Icon Object Reference=Get WBP_CosmeticGating_WarningIcon() -> (WBP_CosmeticGating_WarningIcon:WBP Cosmetic Gating Warning Icon Object Reference)
    In Brush Image Size:Vector 2D Structure=Get IconSize() -> (IconSize:Vector 2D Structure)
   ) -> (then)
  ]}
 GetOfferVM {nodes [
   GetOfferVM() {is=title K2Node_FunctionEntry}
   > Return Node!(
    execute
    OfferVM:Fort Item Shop Offer VM Object Reference=Get FortItemShopOfferVM() -> (FortItemShopOfferVM:Fort Item Shop Offer VM Object Reference)
   )
  ]}
 GetTileAnchor {nodes [
   GetTileAnchor() {is=title K2Node_FunctionEntry}
   > Return Node!(execute, Anchor:Vector 2D Structure)
  ]}
 GetTileWidget {nodes [
   GetTileWidget() {is=title K2Node_FunctionEntry}
   > Return Node!(execute, NewParam:User Widget Object Reference)
  ]}
 VM_GrantedItemsUpdated {nodes [
   VM_GrantedItemsUpdated!() -> (then, GrantedItems:Array of Fort Item VM Object References) {is=title K2Node_FunctionEntry}
   > Set OfferHasContentForCurrentContext(OfferHasContentForCurrentContext:bool=IsThereAnyItemForCurrentItemShopContext!(
     self:Fort Item Shop Offer VM Object Reference=Get FortItemShopOfferVM()
     Items:Array of Fort Item VM Object References=Get GrantedItems() -> (GrantedItems:Array of Fort Item VM Object References)
    ) -> (ReturnValue:bool)) -> (Output_Get:bool)
   > n1 = For Each Loop(Array:Array of Fort Item VM Object References=Get GrantedItems()) -> (Array Element:Fort Item VM Object Reference)
   n2 = ConvertCosmeticItemDefinitionsToInterfaces!(
    execute=@n1.Completed
    CosmeticItemDefinitions:Array of Athena Cosmetic Item Definition Object References=Get ItemDefs() -> (ItemDefs:Array of Athena Cosmetic Item Definition Object References)
   ) -> (then, OutCosmeticItemGatingInterfaces:Array of Cosmetic Item Gating Interface Interfaces)
   > n3 = ItemsHaveRatingRestrictions!(
    execute
    LocalPlayer:Local Player Object Reference=GetOwningLocalPlayer!(self:Widget Object Reference) -> (ReturnValue:Local Player Object Reference)
    CosmeticItems:Array of Cosmetic Item Gating Interface Interfaces=@n2
   ) -> (then, ReturnValue:bool)
   > Set Visibility(Visibility=Select(Option 0=Collapsed, Option 1=SelfHitTestInvisible, Index=@n3), self=Get WBP_CosmeticGating_WarningIcon())
   Branch(
    execute=@n1.LoopBody
    Condition=IsModeDisplayableInCurrentItemShopContext!(
     self:Fort Item Shop Offer VM Object Reference=Get FortItemShopOfferVM()
     ProductTag:Gameplay Tag Structure=GetCurrentProduct!(self:Fort Item VM Object Reference=@n1) -> (ReturnValue:Gameplay Tag Structure)
     bOfferHasContentForCurrentContext:bool=Get OfferHasContentForCurrentContext() -> (OfferHasContentForCurrentContext:bool)
    ) -> (ReturnValue:bool)
   )
   > n4 = Cast To AthenaItemVM!(execute, Object=@n1) -> (then, CastFailed, AsAthena Item VM:Athena Item VM Object Reference) {is=title K2Node_DynamicCast}
   > n5 = Cast To AthenaCosmeticItemDefinition!(execute, Object=Get ItemDefinition!(self:Fort Item VM Object Reference=@n4) -> (ItemDefinition:Fort Item Definition Object Reference)) -> (then, CastFailed, AsAthena Cosmetic Item Definition:Athena Cosmetic Item Definition Object Reference) {is=title K2Node_DynamicCast}
   > Add(
    TargetArray:Array of Athena Cosmetic Item Definition Object References=Get ItemDefs()
    NewItem:"Athena Cosmetic Item Definition Object Reference (by ref)"=@n5
   ) {is=title K2Node_CallArrayFunction}
  ]}
]
variables [IconSize {type=Vector2D, default=<32 32>}]
widgets {
 info {
  inheritedWidgetCount=0
  parentClass=&UserWidget
  rootWidgetClass=/Game/Athena/UI/Frontend/CosmeticGating/WBP_CosmeticGating_WarningIcon.WBP_CosmeticGating_WarningIcon_C
  widgetCount=1
 }
 widgets [WBP_CosmeticGating_WarningIcon {
   widgetClassPath=/Game/Athena/UI/Frontend/CosmeticGating/WBP_CosmeticGating_WarningIcon
   bIsVariable=true
   widget=WBP_CosmeticGating_WarningIcon
   properties {iconSize=<32 32>}
  }]
}
