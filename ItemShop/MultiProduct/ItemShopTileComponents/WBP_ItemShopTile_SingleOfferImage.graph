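# Item shop tile component that shows a single offer's image over a background, both loaded through the offer viewmodel.
# The first loaded presentation fills the image material, with offer-driven colors and per-corner rounding from a material parameter collection.
# A border load guard reflects combined background and presentation loading, including broken-image state.
# Hover animates material parameters, while row state and the non-touch Peek animation control ghost-tile and opacity fades.
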WidgetBlueprint /ItemShop/MultiProduct/ItemShopTileComponents/WBP_ItemShopTile_SingleOfferImage
animations [
 Hover {
  binding_count=1
  length_seconds=2
  object=&self:Hover
  sequence {bindings [Image_Current {
     id=#1
     possessed_class=Image
     tracks [{
       type=MovieSceneWidgetMaterialTrack
       sections [{
         type=MovieSceneParameterSection
         end_frame=40
         start_frame=0
         channels [
          Hover {type=MovieSceneScriptingFloatChannel, keys [[0,0],[4,1]]}
          HoverBurst {type=MovieSceneScriptingFloatChannel, keys [[0,0],[2,3],[4,0],[40,1]]}
         ]
        }]
      }]
    }]}
 }
 Unhover {
  binding_count=1
  length_seconds=0.0999
  object=&self:Unhover
  sequence {bindings [Image_Current {
     id=#1
     possessed_class=Image
     tracks [{
       type=MovieSceneWidgetMaterialTrack
       sections [{
         type=MovieSceneParameterSection
         end_frame=1
         start_frame=0
         channels [
          Hover {type=MovieSceneScriptingFloatChannel, keys [[0,1],[1.9973,0]]}
          HoverBurst {type=MovieSceneScriptingFloatChannel, const [0,2]}
         ]
        }]
      }]
    }]}
 }
 Peek {
  binding_count=1
  length_seconds=0.2
  object=&self:Peek
  sequence {bindings [WBP_ItemShopTile_SingleOfferImage_C_20 {
     id=#2
     possessed_class=WBP_ItemShopTile_SingleOfferImage_C
     tracks [{
       type=MovieSceneDoubleTrack
       display_name=Peek Anim Alpha
       sections [{type=MovieSceneDoubleSection, end_frame=4, start_frame=0, channels [None {type=MovieSceneScriptingDoubleChannel, keys [[0,0],[4,1]]}]}]
      }]
    }]}
 }
]
graphs [
 EventGraph {nodes [
   n1 = OnPresentationsLoaded!() -> (OutputDelegate, then, LoadedPresentations:Array of Fort Item Shop Loaded Presentation Structures) {is=title K2Node_CustomEvent}
   > n2 = Set LoadedPresentations(LoadedPresentations:Array of Fort Item Shop Loaded Presentation Structures=@n1.LoadedPresentations) -> (Output_Get:Array of Fort Item Shop Loaded Presentation Structures)
   > n3 = Branch(Condition="integer > integer"(A=Length(TargetArray:Array of Fort Item Shop Loaded Presentation Structures=@n2)))
   > SetPresentationsLoadingState!(execute, self, LoadState:E_ItemShop_LoadGuardState Enum=NewEnumerator1, ShouldClearPresentationsArray:bool=false) -> (then)
   > SetSinglePresentation!(
    execute
    self
    Presentation:Fort Item Shop Loaded Presentation Structure="Get (a ref)"(
     Array:Array of Fort Item Shop Loaded Presentation Structures=Get LoadedPresentations() -> (LoadedPresentations:Array of Fort Item Shop Loaded Presentation Structures)
     Dimension 1=0
    ) -> (Output:"Fort Item Shop Loaded Presentation Structure (by ref)")
   ) -> (then)
   OnSetBrokenImage(execute=@n3.else)
   OnSetBrokenImage() {is=title K2Node_CustomEvent}
   > SetPresentationsLoadingState(LoadState=NewEnumerator2, ShouldClearPresentationsArray=false)
   n4 = Event PreConstruct()
   > Branch(Condition=@n4.IsDesignTime)
   > n5 = GetDynamicMaterial(self=Get Background() -> (Background:Image Object Reference))
   > SetRoundedCorners!(execute, self, MID:Material Instance Dynamic Object Reference=@n5) -> (then)
   > n6 = GetDynamicMaterial(self=Get Image_Current() -> (Image_Current:Image Object Reference))
   > SetRoundedCorners(MID=@n6)
   n7 = Event SetRowState!() -> (OutputDelegate, then, InRowState:E_ItemShop_RowState Enum=NewEnumerator4) {is=title K2Node_Event}
   > Set Row State Implementation!(execute, self, InRowState:E_ItemShop_RowState Enum=@n7.InRowState) -> (then) {is=title K2Node_CallFunction}
   n8 = Event SetPeekState!() -> (OutputDelegate, then, IsPeeked:bool) {is=title K2Node_Event}
   > SetPeekState_Implementation!(execute, self, InIsPeeked:bool=@n8.IsPeeked) -> (then)
   n9 = Event SetHoveredState!() -> (OutputDelegate, then, IsHovered:bool, IsImmediate:bool) {is=title K2Node_Event}
   > n10 = Set Is Hovered(Is Hovered:bool=@n9.IsHovered) -> (Output_Get:bool)
   > n11 = Branch(Condition=@n10)
   > StopAnimation(InAnimation=Get Unhover() -> (Unhover:Widget Animation Object Reference))
   > PlayAnimationForward(InAnimation=Get Hover() -> (Hover:Widget Animation Object Reference))
   StopAnimation(execute=@n11.else, InAnimation=Get Hover())
   > PlayAnimationForward(InAnimation=Get Unhover())
   n12 = Event RequestLoadPresentations!() -> (OutputDelegate, then, DisplayAssetPath:String) {is=title K2Node_Event}
   > LoadPresentations!(execute, self, DisplayAssetPath:String=@n12.DisplayAssetPath) -> (then)
   n13 = Event SetOfferVM!() -> (OutputDelegate, then, OfferVM:Fort Item Shop Offer VM Object Reference) {is=title K2Node_Event}
   > SetFortItemShopOfferVM!(execute, self, Viewmodel:Fort Item Shop Offer VM Object Reference=@n13.OfferVM) -> (then)
   Event ResetTile() {is=title K2Node_Event}
   > ResetTile_Implementation()
   Event InitializeTile() {is=title K2Node_Event}
   > InitializeTile_Implementation()
  ]}
 GetOfferVM {nodes [
   GetOfferVM() {is=title K2Node_FunctionEntry}
   > Return Node!(
    execute
    OfferVM:Fort Item Shop Offer VM Object Reference=Get FortItemShopOfferVM() -> (FortItemShopOfferVM:Fort Item Shop Offer VM Object Reference)
   )
  ]}
 GetOwningWidgetName {nodes [
   GetOwningWidgetName() {is=title K2Node_FunctionEntry}
   n1 = GetOuterObject!(Object:Object Reference=GetParent!(self:Widget Object Reference) -> (ReturnValue:Panel Widget Object Reference)) -> (ReturnValue:Object Reference)
   > Return Node!(execute, ReturnValue:String=GetDisplayName(Object=GetOuterObject(Object=@n1)))
  ]}
 GetTileAnchor {nodes [
   GetTileAnchor() {is=title K2Node_FunctionEntry}
   > Return Node!(execute, Anchor:Vector 2D Structure)
  ]}
 GetTileWidget {nodes [
   GetTileWidget() {is=title K2Node_FunctionEntry}
   > Return Node!(execute, NewParam:User Widget Object Reference)
  ]}
 InitializeTile_Implementation {nodes [
   InitializeTile_Implementation() {is=title K2Node_FunctionEntry}
   > SetPresentationsLoadingState(LoadState=NewEnumerator0, ShouldClearPresentationsArray=true)
  ]}
 LoadPresentations {nodes [
   LoadPresentations!() -> (then, DisplayAssetPath:String) {is=title K2Node_FunctionEntry}
   > SetPresentationsLoadingState(LoadState=NewEnumerator0, ShouldClearPresentationsArray=true)
   > K2Node_Composite_0 = debug logs - Load presentations!(execute) -> (then:Exec) {is=title K2Node_Composite}
   > LoadPresentations!(
    execute
    self:Fort Item Shop Offer VM Object Reference=Get FortItemShopOfferVM()
    Owner:Object Reference
    InDisplayAssetPath:String=Get DisplayAssetPath() -> (DisplayAssetPath:String)
    DefaultMaterial:Material Interface Object Reference=Get DefaultPresentationMaterial() -> (DefaultPresentationMaterial:Material Interface Object Reference)
    bAllowFullTileOverridePresentations:bool=false
    TextureSize:int=Get DefaultSmallTextureSize() -> (DefaultSmallTextureSize:int=0)
    OnLoadedCallback:"Delegate (by ref)"=Create Event()
    OnLoadFailedCallback:"Delegate (by ref)"=Create Event()
   ) -> (then) {is=title K2Node_CallFunction}
  ]}
 debug logs - Load presentations {
  path=LoadPresentations.K2Node_Composite_0.debug logs - Load presentations
  nodes [
   Inputs()
   > n1 = Branch(Condition=Get EnableLogs() -> (EnableLogs:bool))
   > n2 = Get!(execute:Exec) -> (then:Exec, else:Exec, FortItemShopOfferVM:Fort Item Shop Offer VM Object Reference)
   n3 = GetTitle!(self:Fort Base Offer VM Object Reference=@n2) -> (ReturnValue:Text)
   n4 = GetOwningWidgetName!(self) -> (ReturnValue:String) {is=title K2Node_CallFunction}
   > n5 = PrintString!(
    execute
    InString=Append!(A="Tile - Load Presentations for VM: ", B="To String (Text)"(InText=@n3), C:String=" - Object: ", D:String=@n4) -> (ReturnValue)
    bPrintToScreen=false
    bPrintToLog
    TextColor
    Duration
    Key
   ) -> (then)
   Outputs!(then:Exec=[@n6.then, @n5.then, @n1.else])
   n6 = PrintString(execute=@n2.else, InString=Append!(A="Tile - Load Presentations - invalid VM - Object: ", B=@n4) -> (ReturnValue), bPrintToScreen=false)
  ]
 }
 OnBackgroundOverrideLoaded {nodes [
   n1 = OnBackgroundOverrideLoaded!() -> (then, LoadedBackground:Object Reference) {is=title K2Node_FunctionEntry}
   > n2 = Cast To MaterialInterface!(execute, Object=@n1) -> (then, CastFailed, AsMaterial Interface:Material Interface Object Reference) {is=title K2Node_DynamicCast}
   > SetBackgroundMaterial!(execute, self, Material:Material Interface Object Reference=@n2) -> (then)
   n3 = Cast To Texture2D(execute=@n2.CastFailed, Object=@n1)
   > SetBrushFromMaterial(
    self=Get Background()
    Material=Get DefaultBackgroundTextureMaterial() -> (DefaultBackgroundTextureMaterial:Material Interface Object Reference)
   )
   > n4 = GetDynamicMaterial(self=Get Background())
   > SetTextureParameterValue(self=@n4, ParameterName=Texture, Value=@n3)
   > SetBackgroundLoadingState!(execute, self, IsLoading:bool=false) -> (then)
   SetBackgroundMaterial(
    execute=@n3.CastFailed
    Material=Get DefaultBackgroundGradientMaterial() -> (DefaultBackgroundGradientMaterial:Material Interface Object Reference)
   )
  ]}
 ResetTile_Implementation {nodes [
   ResetTile_Implementation() {is=title K2Node_FunctionEntry}
   n1 = Get FortItemShopOfferVM()
   > CancelLoadingBackground!(execute, self:Fort Item Shop Offer VM Object Reference=@n1, Owner:Object Reference) -> (then)
   > CancelLoadingPresentations!(execute, self:Fort Item Shop Offer VM Object Reference=@n1, Owner:Object Reference) -> (then)
   > ReleaseResource(self:Fort Item Shop Offer VM Object Reference=@n1)
   > SetPresentationsLoadingState(LoadState=NewEnumerator0, ShouldClearPresentationsArray=true)
   > K2Node_Composite_0 = debug logs - reset tile!(execute) -> (then:Exec) {is=title K2Node_Composite}
  ]}
 debug logs - reset tile {
  path=ResetTile_Implementation.K2Node_Composite_0.debug logs - reset tile
  nodes [
   Inputs()
   > n1 = Branch(Condition=Get EnableLogs())
   > n2 = Get()
   n3 = GetOwningWidgetName()
   n4 = GetTitle(self=@n2)
   > n5 = PrintString(
    InString=Append!(A="Reset tile for VM: ", B="To String (Text)"(InText=@n4), C:String=" - Object: ", D:String=@n3) -> (ReturnValue)
    bPrintToScreen=false
   )
   Outputs(then=[@n6.then, @n5.then, @n1.else])
   n6 = PrintString(execute=@n2.else, InString=Append!(A="Reset tile - invalid VM - Object: ", B=@n3) -> (ReturnValue), bPrintToScreen=false)
  ]
 }
 Set Row State Implementation {nodes [
   n1 = Set Row State Implementation!() -> (then, InRowState:E_ItemShop_RowState Enum=NewEnumerator0) {is=title K2Node_FunctionEntry}
   > Set RowState(RowState:E_ItemShop_RowState Enum=@n1) -> (Output_Get:E_ItemShop_RowState Enum=NewEnumerator4)
   > StopAnimation(InAnimation=Select!(
     NewEnumerator4:Widget Animation Object Reference
     NewEnumerator1:Widget Animation Object Reference
     NewEnumerator0:Widget Animation Object Reference=Get Peek() -> (Peek:Widget Animation Object Reference)
     NewEnumerator2:Widget Animation Object Reference
     NewEnumerator3:Widget Animation Object Reference
     Index:E_ItemShop_RowState Enum=Get InRowState() -> (InRowState:E_ItemShop_RowState Enum=NewEnumerator4)
    ) -> (ReturnValue:Widget Animation Object Reference))
   > SetPeekAnimAlpha!(execute, self, Value:double=0) -> (then)
  ]}
 SetBackgroundLoadingState {nodes [
   n1 = SetBackgroundLoadingState!() -> (then, IsLoading:bool=~) {is=title K2Node_FunctionEntry}
   > Set IsLoadingBackground(IsLoadingBackground:bool=@n1) -> (Output_Get:bool)
   > UpdateTileLoadingState()
  ]}
 SetBackgroundMaterial {nodes [
   n1 = SetBackgroundMaterial!() -> (then, Material:Material Interface Object Reference) {is=title K2Node_FunctionEntry}
   > SetBrushFromMaterial(self=Get Background(), Material=@n1)
   > n2 = GetDynamicMaterial(self=Get Background())
   > SetRoundedCorners(MID=@n2)
   n3 = Get FortItemShopOfferVM()
   n4 = Get FilterVM!(self:Fort Item Shop VM Object Reference=Get FortItemShopVM() -> (FortItemShopVM:Fort Item Shop VM Object Reference)) -> (FilterVM:Fort Item Shop Filter VM Object Reference)
   > SetTileBackgroundColors!(
    execute
    MID:Material Instance Dynamic Object Reference=@n2
    FortItemShopOfferMetaDisplayData:"Fort Item Shop Offer Meta Display Data Structure (by ref)"=GetMetaDisplayDataForCurrentViewMode!(
     OfferMetaDisplayData:"Fort Item Shop Offer Meta Display Data Structure (by ref)"=Get MetaDisplayData!(self:Fort Item Shop Offer VM Object Reference=@n3) -> (MetaDisplayData:Fort Item Shop Offer Meta Display Data Structure)
     bIsFilterActive:bool=Get bIsActive!(self:Fort Item Shop Filter VM Object Reference=@n4) -> (bIsActive:bool)
    ) -> (ReturnValue:"Fort Item Shop Offer Meta Display Data Structure (by ref)")
   ) -> (then)
   > SetBackgroundLoadingState(IsLoading=false) {is=title K2Node_CallFunction}
  ]}
 SetHoveredStateVisual {nodes [
   SetHoveredStateVisual!() -> (then, IsHovered:bool=~) {is=title K2Node_FunctionEntry}
   > n1 = GetDynamicMaterial(self=Get Image_Current())
   > Set OfferImageMID(OfferImageMID:Material Instance Dynamic Object Reference=@n1) -> (Output_Get:Material Instance Dynamic Object Reference)
   > n2 = Branch(Condition=Get IsHovered() -> (IsHovered:bool))
   > SetScalarParameterValue(self=Get OfferImageMID() -> (OfferImageMID:Material Instance Dynamic Object Reference), ParameterName=Hover, Value=1)
   > SetScalarParameterValue(self=Get OfferImageMID(), ParameterName=HoverBurst, Value=1)
   SetScalarParameterValue(execute=@n2.else, self=Get OfferImageMID(), ParameterName=Hover, Value=0)
   > SetScalarParameterValue(self=Get OfferImageMID(), ParameterName=HoverBurst, Value=0)
  ]}
 SetLoadGuardState {nodes [
   SetLoadGuardState!() -> (then, State:E_ItemShop_LoadGuardState Enum=NewEnumerator0) {is=title K2Node_FunctionEntry}
   > SetBrushColor(
    self=Get Border_LoadGuard() -> (Border_LoadGuard:Border Object Reference)
    InBrushColor=Select!(
     NewEnumerator0:Linear Color Structure=(R=1,G=1,B=1,A=1)
     NewEnumerator1:Linear Color Structure=(R=1,G=1,B=1,A=0)
     NewEnumerator2:Linear Color Structure=(R=1,G=1,B=1,A=1)
     Index:E_ItemShop_LoadGuardState Enum=Get State() -> (State:E_ItemShop_LoadGuardState Enum=NewEnumerator0)
    ) -> (ReturnValue:Linear Color Structure)
   )
   > Set ContentColorAndOpacity!(
    execute
    ContentColorAndOpacity:Linear Color Structure=Select(
     NewEnumerator0=(R=1,G=1,B=1,A=0)
     NewEnumerator1=(R=1,G=1,B=1,A=1)
     NewEnumerator2=(R=1,G=1,B=1,A=0)
     Index=Get State() -> (State=NewEnumerator0)
    )
    self:Border Object Reference=Get Border_LoadGuard()
   ) -> (then, Output_Get:Linear Color Structure)
   > n1 = GetDynamicMaterial(self:Border Object Reference=Get Border_LoadGuard())
   > SetScalarParameterValue(
    self=@n1
    ParameterName=Is Empty
    Value=Select(NewEnumerator0:float=0, NewEnumerator1:float=0, NewEnumerator2:float=1, Index=Get State() -> (State=NewEnumerator0)) -> (ReturnValue:float)
   )
  ]}
 SetPeekAnimAlpha {nodes [
   SetPeekAnimAlpha!() -> (then, Value:double=~) {is=title K2Node_FunctionEntry}
   > n1 = Switch on E_ItemShop_RowState!(execute, Selection:E_ItemShop_RowState Enum=Get RowState() -> (RowState:E_ItemShop_RowState Enum=NewEnumerator4)) -> (NewEnumerator4:Exec, NewEnumerator1:Exec, NewEnumerator0:Exec, NewEnumerator2:Exec, NewEnumerator3:Exec) {is=title K2Node_SwitchEnum}
   SetRenderOpacity(
    execute=[@n1.NewEnumerator0, @n1.NewEnumerator2]
    self=[Get Image_Current(), Get Background(), Get Overlay_All() -> (Overlay_All:Overlay Object Reference)]
    InOpacity=1
   )
   > SetRenderOpacity(self=Get Image_GhostTile() -> (Image_GhostTile:Image Object Reference), InOpacity=0)
   > n2 = GetDynamicMaterial(self:Image Object Reference=Get Image_Current())
   > SetScalarParameterValue(self=@n2, ParameterName=FadeImageAlpha, Value=0)
   SetRenderOpacity(execute=[@n1.NewEnumerator3, @n1.NewEnumerator4], self=Get Overlay_All(), InOpacity=1)
   > SetRenderOpacity(self=Get Background(), InOpacity=0)
   > SetRenderOpacity(self=Get Image_GhostTile(), InOpacity=Lerp(A=0.8, B=1, Alpha=Get Value() -> (Value:double)))
   > SetRenderOpacity(self=Get Image_Current(), InOpacity=Lerp(A=0.1, B=0.2, Alpha=Get Value()))
   > n3 = GetDynamicMaterial(self=Get Image_Current())
   > SetScalarParameterValue(self=@n3, ParameterName=FadeImageAlpha, Value=1)
   SetRenderOpacity(execute=@n1.NewEnumerator1, self=[Get Background(), Get Image_Current()], InOpacity=1)
   > SetRenderOpacity(self=Get Image_GhostTile(), InOpacity=0)
   > SetRenderOpacity(self=Get Overlay_All(), InOpacity=Lerp(A=0.3, B=0.4, Alpha=Get Value()))
   > n4 = GetDynamicMaterial(self=Get Image_Current())
   > SetScalarParameterValue(self=@n4, ParameterName=FadeImageAlpha, Value=1)
  ]}
 SetPeekState_Implementation {nodes [
   SetPeekState_Implementation!() -> (then, InIsPeeked:bool=~) {is=title K2Node_FunctionEntry}
   > n1 = Branch(Condition=IsUsingTouch())
   n2 = Branch(execute=@n1.else, Condition=Get InIsPeeked() -> (InIsPeeked:bool))
   n3 = Get Peek()
   > PlayAnimationForward(InAnimation=@n3)
   PlayAnimationReverse(execute=@n2.else, InAnimation=@n3)
  ]}
 SetPresentation {nodes [
   n1 = SetPresentation!() -> (then, Image:Image Object Reference, Presentation:"Fort Item Shop Loaded Presentation Structure (by ref)") {is=title K2Node_FunctionEntry}
   > SetBrushFromMaterial(
    self=@n1
    Material=Get!() -> (Presentation_Texture:Texture 2D Object Reference, Presentation_Material:Material Interface Object Reference).Presentation_Material
   )
   > Is Valid(InputObject=Get())
   > n2 = GetDynamicMaterial(self=Get Image() -> (Image:Image Object Reference))
   > n3 = Set MID(MID:Material Instance Dynamic Object Reference=@n2) -> (Output_Get:Material Instance Dynamic Object Reference)
   > SetTextureParameterValue(self=@n3, ParameterName=OfferImage, Value=Get())
   > SetRoundedCorners(MID=Get MID() -> (MID:Material Instance Dynamic Object Reference))
   n4 = Get FortItemShopOfferVM()
   n5 = Get FilterVM(self=Get FortItemShopVM())
   > SetTileBackgroundColors(
    MID=@n3
    FortItemShopOfferMetaDisplayData=GetMetaDisplayDataForCurrentViewMode(OfferMetaDisplayData=Get MetaDisplayData(self=@n4), bIsFilterActive=Get bIsActive(self=@n5))
   )
  ]}
 SetPresentationsLoadingState {nodes [
   n1 = SetPresentationsLoadingState!() -> (then, LoadState:E_ItemShop_LoadGuardState Enum=NewEnumerator0, ShouldClearPresentationsArray:bool=~) {is=title K2Node_FunctionEntry}
   > Set PresentationsLoadState(PresentationsLoadState:E_ItemShop_LoadGuardState Enum=@n1) -> (Output_Get:E_ItemShop_LoadGuardState Enum=NewEnumerator0)
   > n2 = Branch(Condition=@n1.ShouldClearPresentationsArray)
   > n3 = Clear(TargetArray:Array of Fort Item Shop Loaded Presentation Structures=Get LoadedPresentations())
   UpdateTileLoadingState(execute=[@n3.then, @n2.else])
  ]}
 SetRoundedCorners {nodes [
   n1 = SetRoundedCorners!() -> (then, MID:Material Instance Dynamic Object Reference) {is=title K2Node_FunctionEntry}
   > SetScalarParameterValue(self=@n1, ParameterName=bOverrideCornerRadius, Value=1)
   > n2 = GetScalarParameterValue!(
    execute
    Collection:Material Parameter Collection Object Reference=/FortUILibrary/Systems/ItemShopTile/MPC_ItemShop
    ParameterName:Name=CornerRadius
   ) -> (then, ReturnValue:float) {is=title K2Node_CallMaterialParameterCollectionFunction}
   n3 = "To Float (Boolean)"(InBool=Get TopLeft() -> (TopLeft:bool))
   n4 = "To Float (Boolean)"(InBool=Get TopRight() -> (TopRight:bool))
   n5 = "To Float (Boolean)"(InBool=Get BottomLeft() -> (BottomLeft:bool))
   n6 = "To Float (Boolean)"(InBool=Get BottomRight() -> (BottomRight:bool))
   > SetVectorParameterValue(
    self=Get MID()
    ParameterName=CornerRadiusOverride
    Value=LinearColor * LinearColor(
     A:Linear Color Structure=MakeColor!(R:float=@n2, G:float=@n2, B:float=@n2, A:float=@n2) -> (ReturnValue:Linear Color Structure)
     B:Linear Color Structure=MakeColor(R=@n3, G=@n5, B=@n4, A=@n6)
    ) -> (ReturnValue:Linear Color Structure) {is=title K2Node_PromotableOperator}
   )
  ]}
 SetSinglePresentation {nodes [
   n1 = SetSinglePresentation!() -> (then, Presentation:Fort Item Shop Loaded Presentation Structure) {is=title K2Node_FunctionEntry}
   > SetPresentation!(execute, self, Image:Image Object Reference=Get Image_Current(), Presentation:"Fort Item Shop Loaded Presentation Structure (by ref)"=@n1) -> (then) {is=title K2Node_CallFunction}
  ]}
 ShowPresentation {nodes [
   ShowPresentation() {is=title K2Node_FunctionEntry}
   n1 = Get LoadedPresentations()
   > n2 = Branch(Condition="integer > integer"(A=Length(TargetArray=@n1)))
   > SetPresentationsLoadingState(LoadState=NewEnumerator1, ShouldClearPresentationsArray=false) {is=title K2Node_CallFunction}
   > SetSinglePresentation(Presentation="Get (a ref)"(Array=Get LoadedPresentations(), Dimension 1=0)) {is=title K2Node_CallFunction}
   OnSetBrokenImage(execute=@n2.else) {is=title K2Node_CallFunction}
  ]}
 UpdateTileLoadingState {nodes [
   UpdateTileLoadingState() {is=title K2Node_FunctionEntry}
   > SetLoadGuardState!(
    execute
    self
    State:E_ItemShop_LoadGuardState Enum=Select!(
     Option 0:E_ItemShop_LoadGuardState Enum=Get PresentationsLoadState() -> (PresentationsLoadState:E_ItemShop_LoadGuardState Enum=NewEnumerator0)
     Option 1:E_ItemShop_LoadGuardState Enum=NewEnumerator0
     Index:bool=Get IsLoadingBackground() -> (IsLoadingBackground:bool)
    ) -> (ReturnValue:E_ItemShop_LoadGuardState Enum)
   ) -> (then) {is=title K2Node_CallFunction}
  ]}
 VM_DisplayAssetPathChanged {nodes [
   n1 = VM_DisplayAssetPathChanged!() -> (then, DisplayAssetPath:String) {is=title K2Node_FunctionEntry}
   > Branch(Condition=ShouldLoadPresentationsOnTileInit!(self:Fort Item Shop Offer VM Object Reference=Get FortItemShopOfferVM()) -> (ReturnValue:bool))
   > LoadPresentations!(execute, self:Self Object Reference, DisplayAssetPath:String=@n1) -> (then)
  ]}
 VM_MetaDisplayDataUpdated {nodes [
   VM_MetaDisplayDataUpdated!() -> (then, MetaDisplayData:Fort Item Shop Offer Meta Display Data Structure) {is=title K2Node_FunctionEntry}
   > SetBackgroundLoadingState(IsLoading=true)
   > n1 = LoadBackground!(
    execute
    self:Fort Item Shop Offer VM Object Reference=Get FortItemShopOfferVM()
    Owner:Object Reference
    BackgroundKey:String=Break Fort Item Shop Offer Meta Display Data!(FortItemShopOfferMetaDisplayData:"Fort Item Shop Offer Meta Display Data Structure (by ref)"=Get MetaDisplayData!() -> (MetaDisplayData)) -> (TileBackgroundMaterialKey:String) {is=title K2Node_BreakStruct}
    OnLoadedCallback:"Delegate (by ref)"=Create Event()
   ) -> (then, ReturnValue:bool)
   > n2 = Branch(Condition=@n1)
   SetBackgroundMaterial(execute=@n2.else, Material=Get DefaultBackgroundGradientMaterial()) {is=title K2Node_CallFunction}
  ]}
]
variables [
 BottomLeft {type=boolean, default=true}
 BottomRight {type=boolean, default=true}
 DefaultBackgroundGradientMaterial {type=MaterialInterface, default=/FortUILibrary/Systems/ItemShopTile/Materials/M_UI_ItemShopTile_Background_Gradient}
 DefaultBackgroundTextureMaterial {type=MaterialInterface, default=/FortUILibrary/Systems/ItemShopTile/Materials/M_UI_ItemShopTile_Background_Texture}
 DefaultPresentationMaterial {type=MaterialInterface, default=/FortUILibrary/Systems/ItemShopTile/Materials/MI_UI_ItemShopTile_OfferImage_ListView}
 DefaultSmallTextureSize {type=integer, default=114}
 EnableLogs {type=boolean}
 Is Hovered {type=boolean}
 IsLoadingBackground {type=boolean}
 LoadedPresentations {type="array<FortItemShopLoadedPresentation>", default []}
 PeekAnimAlpha {type=number}
 PresentationsLoadState {type=E_ItemShop_LoadGuardState, default=Loading}
 RowState {type=E_ItemShop_RowState, default=None}
 TopLeft {type=boolean, default=true}
 TopRight {type=boolean, default=true}
]
widgets {
 info {inheritedWidgetCount=0, parentClass=&UserWidget, rootWidgetClass=&Border, widgetCount=5}
 widgets [
  Border_LoadGuard {
   widgetClassPath=&Border
   bIsVariable=true
   widget=Border_LoadGuard
   properties {
    background {
     bIsDynamicallyLoaded=false
     drawAs=Image
     imageType=FullColor
     mirroring=NoMirror
     resourceName=None
     resourceObject=/FortUILibrary/Systems/ItemShopTile/Materials/MI_UI_Tile_LoadGuard
     tiling=NoTile
     imageSize=<32 32>
     margin=<ltrb 0 0 0 0>
     outlineSettings {
      bUseBrushTransparency=false
      roundingType=HalfHeightRadius
      width=0
      color {colorUseRule=UseColor_Specified, specifiedColor=<rgba 0 0 0 0>}
      cornerRadii=<0 0 0 0>
     }
     tintColor {colorUseRule=UseColor_Specified, specifiedColor=<rgba 1 1 1 1>}
     uVRegion {bIsValid=false, max=<0 0>, min=<0 0>}
    }
    padding=<ltrb 0 0 0 0>
   }
  }
  Overlay_All {
   widgetClassPath=&Overlay
   bIsVariable=true
   parent=Border_LoadGuard
   widget=Overlay_All
   slot {class=BorderSlot, properties {padding=<ltrb 0 0 0 0>}}
  }
  Background {
   widgetClassPath=&Image
   bIsVariable=true
   parent=Overlay_All
   widget=Background
   properties {
    visibility=SelfHitTestInvisible
    brush {
     bIsDynamicallyLoaded=false
     drawAs=Image
     imageType=FullColor
     mirroring=NoMirror
     resourceName=None
     resourceObject=&M_UI_ItemShopTile_Background_Gradient
     tiling=NoTile
     imageSize=<32 32>
     margin=<ltrb 0 0 0 0>
     outlineSettings {
      bUseBrushTransparency=false
      roundingType=HalfHeightRadius
      width=0
      color {colorUseRule=UseColor_Specified, specifiedColor=<rgba 0 0 0 0>}
      cornerRadii=<0 0 0 0>
     }
     tintColor {colorUseRule=UseColor_Specified, specifiedColor=<rgba 1 1 1 1>}
     uVRegion {bIsValid=false, max=<0 0>, min=<0 0>}
    }
   }
   slot {class=OverlaySlot, properties {horizontalAlignment=HAlign_Fill, verticalAlignment=VAlign_Fill}}
  }
  Image_Current {
   widgetClassPath=&Image
   bIsVariable=true
   parent=Overlay_All
   widget=Image_Current
   properties {
    visibility=SelfHitTestInvisible
    brush {
     bIsDynamicallyLoaded=false
     drawAs=Image
     imageType=FullColor
     mirroring=NoMirror
     resourceName=None
     resourceObject=/FortUILibrary/Systems/ItemShopTile/Materials/M_UI_ItemShopTile_OfferImage_Default
     tiling=NoTile
     imageSize=<32 32>
     margin=<ltrb 0 0 0 0>
     outlineSettings {
      bUseBrushTransparency=false
      roundingType=HalfHeightRadius
      width=0
      color {colorUseRule=UseColor_Specified, specifiedColor=<rgba 0 0 0 0>}
      cornerRadii=<0 0 0 0>
     }
     tintColor {colorUseRule=UseColor_Specified, specifiedColor=<rgba 1 1 1 1>}
     uVRegion {bIsValid=false, max=<0 0>, min=<0 0>}
    }
   }
   slot {class=OverlaySlot, properties {horizontalAlignment=HAlign_Fill, verticalAlignment=VAlign_Fill}}
  }
  Image_GhostTile {
   widgetClassPath=&Image
   bIsVariable=true
   parent=Overlay_All
   widget=Image_GhostTile
   properties {
    renderOpacity=0
    visibility=SelfHitTestInvisible
    brush {
     bIsDynamicallyLoaded=false
     drawAs=Image
     imageType=FullColor
     mirroring=NoMirror
     resourceName=None
     resourceObject=/FortUILibrary/Systems/ItemShopTile/Materials/MI_UI_ItemShop_GhostTile
     tiling=NoTile
     imageSize=<32 32>
     margin=<ltrb 0 0 0 0>
     outlineSettings {
      bUseBrushTransparency=false
      roundingType=HalfHeightRadius
      width=0
      color {colorUseRule=UseColor_Specified, specifiedColor=<rgba 0 0 0 0>}
      cornerRadii=<0 0 0 0>
     }
     tintColor {colorUseRule=UseColor_Specified, specifiedColor=<rgba 1 1 1 1>}
     uVRegion {bIsValid=false, max=<0 0>, min=<0 0>}
    }
   }
   slot {class=OverlaySlot, properties {horizontalAlignment=HAlign_Fill, verticalAlignment=VAlign_Fill}}
  }
 ]
}
