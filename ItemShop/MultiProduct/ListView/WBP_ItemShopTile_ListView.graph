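# List-view row tile for Item Shop song offers, showing image, title, artist, details, highlight and price.
# List items set the offer viewmodel, and granted-item updates supply item details; child components register with the tile base.
# Clicking features the offer and fires analytics, while non-mouse selection drives hover.
# Row state animates opacity, and backplate brightness follows the featured layout's background when dynamic backgrounds are enabled.
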WidgetBlueprint /ItemShop/MultiProduct/ListView/WBP_ItemShopTile_ListView
animations [SetActive {
  binding_count=1
  length_seconds=0.2
  object=&self:SetActive
  sequence {bindings [Overlay_4 {
     id=#1
     possessed_class=Overlay
     tracks [{
       type=MovieSceneFloatTrack
       display_name=Render Opacity
       sections [{type=MovieSceneFloatSection, end_frame=4, start_frame=0, channels [None {type=MovieSceneScriptingFloatChannel, keys [[0,0.3],[4,1]]}]}]
      }]
    }]}
 }]
graphs [
 EventGraph {nodes [
   n1 = InputMethodChanged!() -> (OutputDelegate, then, bNewInputType:ECommonInputType Enum) {is=title K2Node_CustomEvent}
   > Branch(Condition=IsListItemSelected!(UserListEntry:User List Entry Interface) -> (ReturnValue:bool))
   > SetHoveredState!(execute, self:WBP Item Shop Tile Base Object Reference, IsHovered:bool="Equal (Enum)"(A=@n1.bNewInputType, B=Touch), IsImmediate:bool=true) -> (then)
   Event On Selected()
   n2 = "Equal (Enum)"(A=GetCurrentInputType(self=Get CommonInputSubsystem()), B=MouseAndKeyboard)
   > n3 = Branch(Condition=@n2)
   SetHoveredState(execute=@n3.else, self:BPI Item Shop Tile Interface, IsHovered=true, IsImmediate=false)
   Event On Clicked()
   > Set FeaturedOfferId!(
    execute
    FeaturedOfferId:String=Get OfferId!(self:Fort Item Shop Offer VM Object Reference=Get FortItemShopOfferVM() -> (FortItemShopOfferVM:Fort Item Shop Offer VM Object Reference)) -> (OfferId:String)
    self:Fort Item Shop VM Object Reference=Get FortItemShopVM() -> (FortItemShopVM:Fort Item Shop VM Object Reference)
   ) -> (then, Output_Get:String)
   > n4 = Cast To FortPlayerController(Object=GetOwningPlayer())
   n5 = GetDisplayAssetName!(self:Fort Item Shop Offer VM Object Reference=Get FortItemShopOfferVM()) -> (ReturnValue:String)
   > Branch(Condition=NOT Boolean(A=IsEmpty(InString=@n5)))
   > "Fire Event Element Interacted (Non-Localized)"!(execute, FortPC:Fort Player Controller Object Reference=@n4, ElementName:String=@n5, AdditionalInfo:Flow Analytics Additional Info Structure) -> (then)
   Event On Deselected()
   > n6 = Branch(Condition=@n2)
   SetHoveredState(execute=@n6.else, IsHovered=false, IsImmediate=false)
   n7 = Event OnListItemObjectSet()
   > n8 = Cast To FortItemShopOfferVM!(execute, Object=@n7.ListItemObject) -> (then, CastFailed, AsFort Item Shop Offer VM:Fort Item Shop Offer VM Object Reference) {is=title K2Node_DynamicCast}
   > InitializeTile(self:BPI Item Shop Tile Interface)
   > SetOfferVM!(execute, self:BPI Item Shop Tile Interface, OfferVM:Fort Item Shop Offer VM Object Reference=@n8) -> (then)
   n9 = Event PreConstruct()
   > SetupComponents()
   > Parent: PreConstruct(IsDesignTime=@n9.IsDesignTime)
   Event Construct()
   > Bind Event to On Input Method Changed(self=Get CommonInputSubsystem(), Delegate=@n1)
   n10 = Event SetOfferVM!() -> (OutputDelegate, then, OfferVM:Fort Item Shop Offer VM Object Reference) {is=title K2Node_Event}
   > SetFortItemShopOfferVM!(execute, self, Viewmodel:Fort Item Shop Offer VM Object Reference=@n10.OfferVM) -> (then)
   > Parent: SetOfferVM!(execute, OfferVM:Fort Item Shop Offer VM Object Reference=@n10.OfferVM) -> (then) {is=title K2Node_CallParentFunction}
   Event InitializeTile() {is=title K2Node_Event}
   > SetOfferVM(self:WBP Item Shop Tile Base Object Reference)
   > SetFortItemDetailsVM!(execute, self, Viewmodel:Fort Item Details VM Object Reference) -> (then)
   > SetRowState!(execute, self:BPI Item Shop Tile Interface, InRowState:E_ItemShop_RowState Enum=NewEnumerator0) -> (then)
   > SetBackplateBrightness()
   > Parent: InitializeTile() {is=title K2Node_CallParentFunction}
   n11 = Event SetRowState!() -> (OutputDelegate, then, InRowState:E_ItemShop_RowState Enum=NewEnumerator4) {is=title K2Node_Event}
   > n12 = Set RowState(RowState:E_ItemShop_RowState Enum=@n11.InRowState) -> (Output_Get:E_ItemShop_RowState Enum=NewEnumerator4)
   > Parent: SetRowState!(execute, InRowState:E_ItemShop_RowState Enum=@n12) -> (then) {is=title K2Node_CallParentFunction}
   > n13 = Switch on E_ItemShop_RowState!(execute, Selection:E_ItemShop_RowState Enum=@n12) -> (NewEnumerator4:Exec, NewEnumerator1:Exec, NewEnumerator0:Exec, NewEnumerator2:Exec, NewEnumerator3:Exec) {is=title K2Node_SwitchEnum}
   n14 = Get SetActive() -> (SetActive:Widget Animation Object Reference)
   PlayAnimationForward(execute=@n13.NewEnumerator0, InAnimation=@n14)
   PlayAnimationReverse(execute=@n13.NewEnumerator4, InAnimation=@n14)
   Event Destruct()
   > Unbind Event from On Input Method Changed(self=Get CommonInputSubsystem(), Delegate=@n1)
  ]}
 GetOfferVM {nodes [
   GetOfferVM() {is=title K2Node_FunctionEntry}
   > Return Node!(execute, OfferVM:Fort Item Shop Offer VM Object Reference=Get FortItemShopOfferVM())
  ]}
 GetTileWidget {nodes [
   GetTileWidget() {is=title K2Node_FunctionEntry}
   > Return Node!(execute, NewParam:User Widget Object Reference=Self-Reference())
  ]}
 SetBackplateBrightness {nodes [
   SetBackplateBrightness() {is=title K2Node_FunctionEntry}
   > Branch(Condition=AreDynamicBackgroundsEnabledInPdp!(self:Fort Item Shop VM Object Reference=Get FortItemShopVM()) -> (ReturnValue:bool))
   n1 = Get FortItemShopVM()
   n2 = Break Fort Item Shop CMSBackground!(FortItemShopCMSBackground:"Fort Item Shop CMSBackground Structure (by ref)"=GetPreferredBackgroundForLayout!(
     self:Fort Item Shop VM Object Reference=@n1
     InLayout:"Fort Item Shop Layout Structure (by ref)"=Get FeaturedLayout!(self:Fort Item Shop VM Object Reference=@n1) -> (FeaturedLayout:Fort Item Shop Layout Structure)
    ) -> (ReturnValue:Fort Item Shop CMSBackground Structure)) -> (CustomTexture:String, CookedAssetKey:String, Type:EFortItemShopBackgroundType Enum) {is=title K2Node_BreakStruct}
   > Branch(Condition="Equal (Enum)"(A:EFortItemShopBackgroundType Enum=@n2.Type, B:EFortItemShopBackgroundType Enum=Light))
   > SetBrightness!(
    execute
    self:WBP UIKit Backplate Base Object Reference=Get WBP_UIKit_Backplate() -> (WBP_UIKit_Backplate:WBP UIKit Backplate Object Reference)
    Brightness:E_UI_BackplateBrightness Enum=Select!(
     Light:E_UI_BackplateBrightness Enum=NewEnumerator2
     Dark:E_UI_BackplateBrightness Enum=NewEnumerator0
     Index:EFortItemShopBackgroundType Enum=@n2.Type
    ) -> (ReturnValue:E_UI_BackplateBrightness Enum)
   ) -> (then)
  ]}
 SetupComponents {nodes [
   SetupComponents() {is=title K2Node_FunctionEntry}
   > RegisterComponents!(
    execute
    self:WBP Item Shop Tile Base Object Reference
    Components:Array of BPI Item Shop Tile Interfaces=Make Array!(
     "[0]":BPI Item Shop Tile Interface=Get Component_Price() -> (Component_Price:WBP Item Shop Tile Price Object Reference)
     "[1]":BPI Item Shop Tile Interface=Get Component_Highlight() -> (Component_Highlight:WBP Item Shop Tile Highlight 2 Object Reference)
     "[2]":BPI Item Shop Tile Interface=Get Component_3DPreview() -> (Component_3DPreview:WBP Item Shop Tile 3DPreview Object Reference)
     "[3]":BPI Item Shop Tile Interface=Get Component_CosmeticGatingWarningIcon() -> (Component_CosmeticGatingWarningIcon:WBP Item Shop Tile Cosmetic Gating Warning Icon Object Reference)
     "[4]":BPI Item Shop Tile Interface=Get Component_SingleOfferImage() -> (Component_SingleOfferImage:WBP Item Shop Tile Single Offer Image Object Reference)
    ) -> (Array:Array of BPI Item Shop Tile Interfaces)
   ) -> (then)
   > RegisterModularBlock!(
    execute
    self:UIKit Modular Button Object Reference
    Block:UIKit Block Object Reference=Get ModularBlock_Outline() -> (ModularBlock_Outline:WBP UIKit Block Outline Object Reference)
   ) -> (then)
   > SetBlocksTiming!(
    execute
    self:UIKit Modular Button Object Reference
    Timing_Hovering:float=0
    Timing_Unhovering:float=0
    Timing_Focusing:float=0.1
    Timing_Unfocusing:float=0
    Timing_Pressing:float=0.1
    Timing_Releasing:float=0
    Timing_Disabling:float=0
    Timing_Enabling:float=0
    Timing_Selecting:float=0
    Timing_DeselectingIdle:float=0
    Timing_DeselectingFocused:float=0
   ) -> (then)
  ]}
 VM_GrantedItemsUpdated {nodes [
   n1 = VM_GrantedItemsUpdated!() -> (then, GrantedItems:Array of Fort Item VM Object References) {is=title K2Node_FunctionEntry}
   > Branch(Condition=Is Not Empty!(TargetArray:Array of Fort Item VM Object References=@n1) -> (ReturnValue:bool))
   > SetFortItemDetailsVM(Viewmodel=Get ItemDetails!(self:Fort Item VM Object Reference="Get (a copy)"(Array:Array of Fort Item VM Object References=@n1) -> (Output:Fort Item VM Object Reference)) -> (ItemDetails:Fort Item Details VM Object Reference))
  ]}
]
variables [RowState {type=E_ItemShop_RowState, default=None}]
widgets {
 info {
  inheritedWidgetCount=1
  parentClass=/ItemShop/MultiProduct/ItemShopTile/WBP_ItemShopTile_Base.WBP_ItemShopTile_Base_C
  rootWidgetClass=&Overlay
  widgetCount=19
 }
 widgets [
  Overlay_4 {widgetClassPath=&Overlay, widget=Overlay_4}
  WBP_UIKit_Backplate {
   widgetClassPath=&WBP_UIKit_Backplate
   bIsVariable=true
   parent=Overlay_4
   widget=WBP_UIKit_Backplate
   properties {brightness=Medium, cornerRadius=Interactive Component}
   slot {class=OverlaySlot, properties {horizontalAlignment=HAlign_Fill, verticalAlignment=VAlign_Fill}}
  }
  Overlay_1 {
   widgetClassPath=&Overlay
   parent=Overlay_4
   widget=Overlay_1
   slot {class=OverlaySlot, properties {horizontalAlignment=HAlign_Fill, verticalAlignment=VAlign_Center}}
  }
  HorizontalBox {
   widgetClassPath=&HorizontalBox
   parent=Overlay_1
   widget=HorizontalBox
   properties {clipping=ClipToBounds}
   slot {class=OverlaySlot, properties {horizontalAlignment=HAlign_Fill}}
  }
  Overlay_Image {widgetClassPath=&Overlay, bIsVariable=true, parent=HorizontalBox, widget=Overlay_Image, slot {class=HorizontalBoxSlot}}
  SizeBox_Image {
   widgetClassPath=&SizeBox
   bIsVariable=true
   parent=Overlay_Image
   widget=SizeBox_Image
   properties {bOverride_HeightOverride=true, bOverride_WidthOverride=true, heightOverride=114, widthOverride=114}
   slot {class=OverlaySlot, properties {horizontalAlignment=HAlign_Fill, verticalAlignment=VAlign_Fill}}
  }
  Component_SingleOfferImage {
   widgetClassPath=/ItemShop/MultiProduct/ItemShopTileComponents/WBP_ItemShopTile_SingleOfferImage
   bIsVariable=true
   parent=SizeBox_Image
   widget=Component_SingleOfferImage
   properties {bottomRight=false, topRight=false}
   slot {class=SizeBoxSlot}
  }
  Component_CosmeticGatingWarningIcon {
   widgetClassPath=/ItemShop/MultiProduct/ItemShopTileComponents/WBP_ItemShopTile_CosmeticGatingWarningIcon
   bIsVariable=true
   parent=Overlay_Image
   widget=Component_CosmeticGatingWarningIcon
   slot {class=OverlaySlot, properties {padding=<ltrb 8 8 8 8>}}
  }
  Overlay_OfferInfo {
   widgetClassPath=&Overlay
   bIsVariable=true
   parent=HorizontalBox
   widget=Overlay_OfferInfo
   slot {class=HorizontalBoxSlot, properties {size {sizeRule=Fill, value=1}}}
  }
  VerticalBox_OfferInfo {
   widgetClassPath=&VerticalBox
   bIsVariable=true
   parent=Overlay_OfferInfo
   widget=VerticalBox_OfferInfo
   slot {class=OverlaySlot, properties {horizontalAlignment=HAlign_Fill, verticalAlignment=VAlign_Center, padding=<ltrb 12 0 12 0>}}
  }
  Song_Name {
   widgetClassPath=&CommonTextBlock
   bIsVariable=true
   parent=VerticalBox_OfferInfo
   widget=Song_Name
   properties {
    bIsScrollingEnabled=false
    clipping=OnDemand
    lineHeightPercentage=0.9
    mobileFontSizeMultiplier=1.2
    scrollStyle=&TextScrollStyle-Base_C
    text="[PH] Song name long very long very long"
    textOverflowPolicy=Ellipsis
    font {
     bForceMonospaced=false
     bMaterialIsStencil=false
     fontMaterial=/FortUILibrary/UIKit/Buttons/MaterialInstances/ItemShopTile/MI_UIKit_ItemShopTile_TitleText
     fontObject=&HeadingNow
     letterSpacing=0
     monospacedWidth=1
     size=20.25
     skewAmount=0
     typefaceFontName=75 Medium
     outlineSettings {
      bApplyOutlineToDropShadows=false
      bMiteredCorners=false
      bSeparateFillAlpha=false
      outlineMaterial=null
      outlineSize=0
      outlineColor=<rgba 0 0 0 1>
     }
    }
    shadowOffset=<0 0>
   }
   slot {class=VerticalBoxSlot, properties {verticalAlignment=VAlign_Top}}
  }
  Artist_Name {
   widgetClassPath=&CommonTextBlock
   bIsVariable=true
   parent=VerticalBox_OfferInfo
   widget=Artist_Name
   properties {
    bIsScrollingEnabled=false
    mobileFontSizeMultiplier=1.2
    scrollStyle=&TextScrollStyle-Base_C
    text="[PH] Artist Name long very very long"
    textOverflowPolicy=Ellipsis
    font {
     bForceMonospaced=false
     bMaterialIsStencil=false
     fontMaterial=null
     fontObject=&HeadingNow
     letterSpacing=0
     monospacedWidth=1
     size=15.75
     skewAmount=0
     typefaceFontName=74 Regular
     outlineSettings {
      bApplyOutlineToDropShadows=false
      bMiteredCorners=false
      bSeparateFillAlpha=false
      outlineMaterial=null
      outlineSize=0
      outlineColor=<rgba 0 0 0 1>
     }
    }
   }
   slot {class=VerticalBoxSlot, properties {verticalAlignment=VAlign_Center, padding=<ltrb 0 5 0 4>}}
  }
  Song_Details {
   widgetClassPath=&CommonTextBlock
   bIsVariable=true
   parent=VerticalBox_OfferInfo
   widget=Song_Details
   properties {
    bIsScrollingEnabled=false
    mobileFontSizeMultiplier=1.2
    scrollStyle=&TextScrollStyle-Base_C
    text="[PH] 2019 | 03:24"
    textOverflowPolicy=Ellipsis
    colorAndOpacity {colorUseRule=UseColor_Specified, specifiedColor=<rgba 1 1 1 0.65>}
    font {
     bForceMonospaced=false
     bMaterialIsStencil=false
     fontMaterial=null
     fontObject=&HeadingNow
     letterSpacing=0
     monospacedWidth=1
     size=14.25
     skewAmount=0
     typefaceFontName=74 Regular
     outlineSettings {
      bApplyOutlineToDropShadows=false
      bMiteredCorners=false
      bSeparateFillAlpha=false
      outlineMaterial=null
      outlineSize=0
      outlineColor=<rgba 0 0 0 1>
     }
    }
   }
   slot {class=VerticalBoxSlot, properties {verticalAlignment=VAlign_Bottom}}
  }
  Overlay_0 {
   widgetClassPath=&Overlay
   parent=HorizontalBox
   widget=Overlay_0
   slot {class=HorizontalBoxSlot, properties {padding=<ltrb 10 10 10 10>}}
  }
  Component_Highlight {
   widgetClassPath=/ItemShop/MultiProduct/ItemShopTileComponents/WBP_ItemShopTile_Highlight_2
   bIsVariable=true
   parent=Overlay_0
   widget=Component_Highlight
   properties {highlightText="[PH] Highlight", widthOverride=200}
   slot {class=OverlaySlot, properties {horizontalAlignment=HAlign_Right}}
  }
  Component_Price {
   widgetClassPath=/ItemShop/MultiProduct/ItemShopTileComponents/WBP_ItemShopTile_Price
   bIsVariable=true
   parent=Overlay_0
   widget=Component_Price
   slot {class=OverlaySlot, properties {horizontalAlignment=HAlign_Right, verticalAlignment=VAlign_Bottom}}
  }
  ModularBlock_Outline {
   widgetClassPath=/FortUILibrary/UIKit/Buttons/Blocks/WBP_UIKit_Block_Outline
   bIsVariable=true
   parent=Overlay_1
   widget=ModularBlock_Outline
   properties {bottomLeftCornerRadius=27, bottomRightCornerRadius=27, radiusInPixels=true, topLeftCornerRadius=27, topRightCornerRadius=27}
   slot {class=OverlaySlot, properties {horizontalAlignment=HAlign_Fill, verticalAlignment=VAlign_Fill}}
  }
  Component_3DPreview {
   widgetClassPath=/ItemShop/MultiProduct/ItemShopTileComponents/WBP_ItemShopTile_3DPreview
   bIsVariable=true
   parent=Overlay_4
   widget=Component_3DPreview
   slot {class=OverlaySlot}
  }
  InputActionWidget {widgetClassPath=&CommonActionWidget, bInherited=true, widget=null}
 ]
}
