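# Item Shop list-view price widget that holds a price display and an "Owned" label in a visibility switcher.
# The price box stacks WBP_Price_ItemShop with discount-disclaimer and real-price text blocks, which are bound to its optional text properties in the widget tree.
# PreConstruct forwards RealPriceLegalText_Optional, if valid, to the inner price widget.
# The graph contains no logic for choosing between the price and Owned states.
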WidgetBlueprint /ItemShop/MultiProduct/ListView/WBP_ItemShop_ListView_Price
graphs [EventGraph {nodes [
   Event PreConstruct()
   > n1 = Get!(execute:Exec) -> (then:Exec, else:Exec, RealPriceLegalText_Optional:Common Text Object Reference)
   > Set RealPriceLegalText_Optional!(
    execute
    RealPriceLegalText_Optional:Common Text Object Reference=@n1
    self:WBP UIKit Price Object Reference=Get WBP_Price_ItemShop() -> (WBP_Price_ItemShop:WBP Price Item Shop Object Reference)
   ) -> (then, Output_Get:Common Text Object Reference)
  ]}]
variables [RealPriceLegalText_Optional {type=CommonTextBlock, default=null}]
widgets {
 info {inheritedWidgetCount=0, parentClass=&UserWidget, rootWidgetClass=&CommonVisibilitySwitcher, widgetCount=6}
 widgets [
  CommonVisibilitySwitcher_0 {widgetClassPath=&CommonVisibilitySwitcher, widget=CommonVisibilitySwitcher_0}
  VerticalBox_Price {
   widgetClassPath=&VerticalBox
   parent=CommonVisibilitySwitcher_0
   widget=VerticalBox_Price
   properties {bOverride_Cursor=true}
   slot {class=CommonVisibilitySwitcherSlot, properties {horizontalAlignment=HAlign_Center, verticalAlignment=VAlign_Top}}
  }
  WBP_Price_ItemShop {
   widgetClassPath=/ItemShop/MultiProduct/WBP_Price_ItemShop
   parent=VerticalBox_Price
   widget=WBP_Price_ItemShop
   properties {
    discountDisclaimerText_Optional=&self:WidgetTree.DiscountDisclaimer
    realPriceText_Optional=&self:WidgetTree.Text_RealPrice
   }
   slot {class=VerticalBoxSlot, properties {horizontalAlignment=HAlign_Center}}
  }
  DiscountDisclaimer {
   widgetClassPath=&CommonTextBlock
   bIsVariable=true
   parent=VerticalBox_Price
   widget=DiscountDisclaimer
   properties {
    applyLineHeightToBottomLine=false
    bAutoCollapseWithEmptyText=true
    justification=Center
    mobileFontSizeMultiplier=1.2
    text="[PH]*Discount for select new players; full price applies to all others"
    font {
     bForceMonospaced=false
     bMaterialIsStencil=false
     fontMaterial=null
     fontObject=&HeadingNow
     letterSpacing=0
     monospacedWidth=1
     size=11.25
     skewAmount=0
     typefaceFontName=74 Regular
     outlineSettings {
      bApplyOutlineToDropShadows=false
      bMiteredCorners=false
      bSeparateFillAlpha=false
      outlineMaterial=null
      outlineSize=0
      outlineColor=<rgba 0 0 0 1>
     }
    }
    shadowOffset=<0 0>
   }
   slot {class=VerticalBoxSlot}
  }
  Text_RealPrice {
   widgetClassPath=&CommonTextBlock
   bIsVariable=true
   parent=VerticalBox_Price
   widget=Text_RealPrice
   properties {
    bAutoCollapseWithEmptyText=true
    justification=Center
    mobileFontSizeMultiplier=1.2
    text=(equivalent to $8.99*)
    font {
     bForceMonospaced=false
     bMaterialIsStencil=false
     fontMaterial=null
     fontObject=&HeadingNow
     letterSpacing=0
     monospacedWidth=1
     size=18
     skewAmount=0
     typefaceFontName=74 Regular
     outlineSettings {
      bApplyOutlineToDropShadows=false
      bMiteredCorners=false
      bSeparateFillAlpha=false
      outlineMaterial=null
      outlineSize=0
      outlineColor=<rgba 0 0 0 1>
     }
    }
    shadowOffset=<0 0>
   }
   slot {class=VerticalBoxSlot, properties {horizontalAlignment=HAlign_Center, padding=<ltrb 0 4 0 0>}}
  }
  Text_Owned {
   widgetClassPath=&CommonTextBlock
   parent=CommonVisibilitySwitcher_0
   widget=Text_Owned
   properties {
    bAutoCollapseWithEmptyText=true
    justification=Center
    mobileFontSizeMultiplier=1.2
    text=Owned
    textTransformPolicy=ToUpper
    colorAndOpacity {colorUseRule=UseColor_Specified, specifiedColor=<rgba 0.0369 1 0.2232 1>}
    font {
     bForceMonospaced=false
     bMaterialIsStencil=false
     fontMaterial=/FortUILibrary/Systems/ItemShopTile/Materials/PriceSwitcher/MI_UIKit_PriceSwitcher_Text
     fontObject=&HeadingNow
     letterSpacing=0
     monospacedWidth=1
     size=24
     skewAmount=0
     typefaceFontName=86 Bold Italic
     outlineSettings {
      bApplyOutlineToDropShadows=false
      bMiteredCorners=false
      bSeparateFillAlpha=false
      outlineMaterial=null
      outlineSize=0
      outlineColor=<rgba 0 0 0 1>
     }
    }
    shadowOffset=<0 0>
   }
   slot {class=CommonVisibilitySwitcherSlot}
  }
 ]
}
