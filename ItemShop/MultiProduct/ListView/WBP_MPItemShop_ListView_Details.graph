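# Item Shop list-view details panel showing one offer's price, warnings, and purchase, Get V-Bucks, gift, undo and view-bundle buttons.
# View-model update functions toggle button and warning visibility; gated offers route through a cosmetic-gate modal.
# Purchases pass through optional confirmation or price-choice steps, then show an overlay or gift box, with interactions disabled meanwhile.
# PreConstruct applies compact-UI layout; hover changes are broadcast, with unhover debounced by a short delay.
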WidgetBlueprint /ItemShop/MultiProduct/ListView/WBP_MPItemShop_ListView_Details
graphs [
 EventGraph {nodes [
   "On Clicked (Button_GetVBucks)"!() -> (OutputDelegate, then, Button:Common Button Base Object Reference) {is=title K2Node_ComponentBoundEvent}
   > RequestOpenMtxStore!(
    execute
    Context:Widget Object Reference
    Offer:Fort Base Offer VM Object Reference=Get FortItemShopOfferVM() -> (FortItemShopOfferVM:Fort Item Shop Offer VM Object Reference)
   ) -> (then)
   n1 = "On Hold Completed (Button_Purchase)"() {is=title K2Node_ComponentBoundEvent}
   SetAllInteractionsEnabled!(execute=[@n1.then, @n14.else], self, ShouldEnable:bool=false) -> (then)
   > n2 = Branch(Condition=Get IsOfferGated() -> (IsOfferGated:bool))
   > ShowCosmeticGateConfirmation()
   AttemptPurchase(execute=@n2.else)
   "On Clicked (Button_Gift)"!() -> (OutputDelegate, then, Button:Common Button Base Object Reference) {is=title K2Node_ComponentBoundEvent}
   > ShowGiftingScreen!(execute, Context:Widget Object Reference, OfferVM:Fort Item Shop Offer VM Object Reference=Get FortItemShopOfferVM(), OnMFAShown:Delegate) -> (then)
   n3 = "On Unhovered (Button_Gift)"!() -> (OutputDelegate, then, Button:Common Button Base Object Reference) {is=title K2Node_ComponentBoundEvent}
   Delay(execute=[@n3.then, @n10.then, @n11.then, @n12.then, @n24.then], Duration=0.1)
   n4 = Get Button_ViewBundle() -> (Button_ViewBundle:WBP UIKit Button Regular Object Reference)
   n5 = Get Button_Purchase() -> (Button_Purchase:WBP UIKit Button Loud Object Reference)
   n6 = Get Button_UndoPurchase() -> (Button_UndoPurchase:WBP UIKit Button Regular Object Reference)
   n7 = Get Button_GetVBucks() -> (Button_GetVBucks:WBP UIKit Button Loud Object Reference)
   n8 = Get Button_Gift() -> (Button_Gift:WBP UIKit Button Regular Object Reference)
   > n9 = Branch(Condition=OR Boolean!(A=IsHovered(self=@n5), B=IsHovered(self=@n6), C:bool=IsHovered(self=@n7), D:bool=IsHovered(self=@n8), E:bool=IsHovered(self=@n4)) -> (ReturnValue))
   Call On Details Buttons Hovered Changed!(execute:Exec=@n9.else, self:WBP MPItem Shop List View Details Object Reference, IsHovered:bool) -> (then:Exec) {is=title K2Node_CallDelegate}
   n10 = "On Unhovered (Button_Purchase)"!() -> (OutputDelegate, then, Button:Common Button Base Object Reference) {is=title K2Node_ComponentBoundEvent}
   n11 = "On Unhovered (Button_UndoPurchase)"!() -> (OutputDelegate, then, Button:Common Button Base Object Reference) {is=title K2Node_ComponentBoundEvent}
   n12 = "On Unhovered (Button_ViewBundle)"!() -> (OutputDelegate, then, Button:Common Button Base Object Reference) {is=title K2Node_ComponentBoundEvent}
   n13 = "On Clicked (Button_Purchase)"!() -> (OutputDelegate, then, Button:Common Button Base Object Reference) {is=title K2Node_ComponentBoundEvent}
   > n14 = Branch(Condition=GetRequiresHold!(self:Common Button Base Object Reference=@n13.Button) -> (ReturnValue:bool))
   "On Hold Completed (Button_UndoPurchase)"() {is=title K2Node_ComponentBoundEvent}
   > UndoPurchase()
   "On Clicked (Button_ViewBundle)"!() -> (OutputDelegate, then, Button:Common Button Base Object Reference) {is=title K2Node_ComponentBoundEvent}
   > n15 = Async Load Class Asset!(execute:Exec, AssetClass:Object Soft Class Reference=/ItemShop/MultiProduct/WBP_ItemShop_PDP.WBP_ItemShop_PDP_C) -> (then:Exec, Completed:Exec, Class:Object Class Reference) {is=title K2Node_LoadAssetClass}
   > SetAllInteractionsEnabled(ShouldEnable=false)
   n16 = Cast To CommonActivatableWidget Class!(execute:Exec=@n15.Completed, Class:Object Class Reference=@n15) -> (then:Exec, CastFailed:Exec, AsCommon Activatable Widget:Common Activatable Widget Class Reference) {is=title K2Node_ClassDynamicCast}
   > n17 = PushPrimaryContentClassAsync!(execute:Exec, WidgetClass:Common Activatable Widget Class Reference=@n16) -> (then:Exec, OnPrimaryContentPushed:Exec, ActivatableWidget:Common Activatable Widget Object Reference) {is=title K2Node_AsyncAction}
   n18 = Cast To FortMPItemShopOfferInspectionScreen!(execute=@n17.OnPrimaryContentPushed, Object=@n17) -> (then, CastFailed, AsFort MPItem Shop Offer Inspection Screen:Fort MPItem Shop Offer Inspection Screen Object Reference) {is=title K2Node_DynamicCast}
   > SetViewModels!(
    execute
    self:Fort MPItem Shop Offer Inspection Screen Object Reference=@n18
    OfferVM:Fort Item Shop Offer VM Object Reference=Get BundleOfferVM() -> (BundleOfferVM:Fort Item Shop Offer VM Object Reference)
    OfferGroupVM:Fort Item Shop Offer Group VM Object Reference
   ) -> (then)
   > SetAllInteractionsEnabled(ShouldEnable=true)
   n19 = "On Hovered (Button_UndoPurchase)"!() -> (OutputDelegate, then, Button:Common Button Base Object Reference) {is=title K2Node_ComponentBoundEvent}
   Call On Details Buttons Hovered Changed(execute=[@n19.then, @n20.then, @n21.then, @n22.then, @n23.then], IsHovered=true)
   n20 = "On Hovered (Button_GetVBucks)"!() -> (OutputDelegate, then, Button:Common Button Base Object Reference) {is=title K2Node_ComponentBoundEvent}
   n21 = "On Hovered (Button_Gift)"!() -> (OutputDelegate, then, Button:Common Button Base Object Reference) {is=title K2Node_ComponentBoundEvent}
   n22 = "On Hovered (Button_Purchase)"!() -> (OutputDelegate, then, Button:Common Button Base Object Reference) {is=title K2Node_ComponentBoundEvent}
   n23 = "On Hovered (Button_ViewBundle)"!() -> (OutputDelegate, then, Button:Common Button Base Object Reference) {is=title K2Node_ComponentBoundEvent}
   n24 = "On Unhovered (Button_GetVBucks)"!() -> (OutputDelegate, then, Button:Common Button Base Object Reference) {is=title K2Node_ComponentBoundEvent}
   n25 = SetCanUndoPurchase!() -> (OutputDelegate, then, CanUndo:bool=~) {is=title K2Node_CustomEvent}
   > n26 = Set Visibility(Visibility=Select(Option 0=Collapsed, Option 1=SelfHitTestInvisible, Index=@n25.CanUndo), self=Get Button_UndoPurchase())
   > Set Visibility(Visibility=@n26, self=Get Text_Legal_UndoPurchase() -> (Text_Legal_UndoPurchase:Common Text Object Reference))
   n27 = SetCanGift!() -> (OutputDelegate, then, IsGiftable:bool=~, IsOwned:bool=~) {is=title K2Node_CustomEvent}
   > Set Visibility(Visibility=Select(Option 0=Collapsed, Option 1=SelfHitTestInvisible, Index=@n27.IsGiftable), self=Get Button_Gift())
   n28 = SetRatingBoardWarning!() -> (OutputDelegate, then, ShowWarning:bool=~, WarningText:Text) {is=title K2Node_CustomEvent}
   > SetText(
    self:WBP UIKit Info Box Object Reference=Get Warning_RatingBoard() -> (Warning_RatingBoard:WBP UIKit Info Box Object Reference)
    InText:Text=@n28.WarningText
   )
   > Set Visibility(Visibility=Select(Option 0=Collapsed, Option 1=SelfHitTestInvisible, Index=@n28.ShowWarning), self=Get Warning_RatingBoard())
   RequestPurchasePinModal() {is=title K2Node_CustomEvent}
   > RequestPurchasePinModal!(
    execute
    Context:Widget Object Reference
    OfferVM:Fort Item Shop Offer VM Object Reference=Get FortItemShopOfferVM()
    OnPinVerified:Delegate=Create Event()
   ) -> (then) {is=title K2Node_CallFunction}
   n29 = OnOfferPurchased!() -> (OutputDelegate, then, bSuccess:bool=~, PurchasedItems:Array of Purchased Item Info Structures, OfferId:String) {is=title K2Node_CustomEvent}
   > n30 = Branch(Condition=@n29.bSuccess)
   > UpdateGrantedItemsIsOwned(self:Fort Item Shop Offer VM Object Reference=Get FortItemShopOfferVM())
   n31 = GetCurrencyType!(self:Fort Base Offer VM Object Reference=Get FortItemShopOfferVM()) -> (ReturnValue:EFortPriceCurrency Enum=Mtx)
   > n32 = Branch(Condition=OR Boolean!(
     A="Equal (Enum)"(A:EFortPriceCurrency Enum=@n31, B:EFortPriceCurrency Enum=RealMoney)
     B=Is Empty!(TargetArray:Array of Purchased Item Info Structures=@n29.PurchasedItems) -> (ReturnValue:bool)
    ) -> (ReturnValue))
   > SetAllInteractionsEnabled(ShouldEnable=true)
   n33 = Branch(execute=@n32.else, Condition=ShouldShowOfferPurchasedOverlay!() -> (ReturnValue:bool))
   > n34 = GetItemTemplateIDs!(execute, self, ItemInfo:Array of Purchased Item Info Structures=@n29.PurchasedItems) -> (then, ItemTemplateIDs:Array of Strings)
   > ShowPurchaseOverlay!(execute, self, PurchasedItemTemplateIDs:Array of Strings=@n34, ShowAsRefund:bool=false) -> (then)
   n35 = PushGiftBoxScreenFromPurchaseInfo!(
    execute:Exec=@n33.else
    InPurchasedItems:Array of Purchased Item Info Structures=@n29.PurchasedItems
    bRemoveDuplicatesAndVariantInfo:bool=false
   ) -> (then:Exec, OnPrimaryContentPushed:Exec, ActivatableWidget:Common Activatable Widget Object Reference) {is=title K2Node_AsyncAction}
   Assign On Widget Deactivated!(execute:Exec=@n35.OnPrimaryContentPushed, self:Common Activatable Widget Object Reference=@n35, Delegate:"Delegate (by ref)"=Create Event()) -> (then:Exec) {is=title K2Node_AssignDelegate}
   SetAllInteractionsEnabled(execute=@n30.else, ShouldEnable=true)
   n36 = SetIsGated!() -> (OutputDelegate, then, IsGated:bool=~, DisclaimerText:Text) {is=title K2Node_CustomEvent}
   > n37 = Set IsOfferGated(IsOfferGated:bool=@n36.IsGated) -> (Output_Get:bool)
   > SetText(self=Get Warning_Cosmetic() -> (Warning_Cosmetic:WBP UIKit Info Box Object Reference), InText=@n36.DisclaimerText)
   > Set Visibility(Visibility=Select(Option 0=Collapsed, Option 1=SelfHitTestInvisible, Index=@n37), self=Get Warning_Cosmetic())
   ShowPurchaseConfirmation() {is=title K2Node_CustomEvent}
   > n38 = Construct Object from Class(Class=/ItemShop/MultiProduct/Data/DialogVM_PurchaseConfirmation.DialogVM_PurchaseConfirmation_C) -> (ReturnValue:Dialog VM Purchase Confirmation Object Reference)
   > n39 = Create Widget!(
    execute:Exec
    Class:User Widget Class Reference=/ItemShop/MultiProduct/WBP_PurchaseConfirmation_Dialog.WBP_PurchaseConfirmation_Dialog_C
    OwningPlayer:Player Controller Object Reference=GetOwningPlayer()
    DialogVM_PurchaseConfirmation:Dialog VM Purchase Confirmation Object Reference=@n38
    FortItemShopOfferVM:Fort Item Shop Offer VM Object Reference=Get FortItemShopOfferVM()
   ) -> (then:Exec, ReturnValue:WBP Purchase Confirmation Dialog Object Reference)
   n40 = OnPurchaseConfirmed() {is=title K2Node_CustomEvent}
   Unbind Event from Purchase Confirmed(execute=@n39.then, self:WBP Purchase Confirmation Dialog Object Reference=@n39, Delegate=@n40) {is=title K2Node_RemoveDelegate}
   > n41 = Bind Event to Purchase Confirmed(self:WBP Purchase Confirmation Dialog Object Reference=@n39, Delegate=@n40) {is=title K2Node_AddDelegate}
   n42 = EnableAllInteractions() {is=title K2Node_CustomEvent}
   Unbind Event from Purchase Dismissed(execute=@n41.then, self:WBP Purchase Confirmation Dialog Object Reference=@n39, Delegate=@n42) {is=title K2Node_RemoveDelegate}
   > Bind Event to Purchase Dismissed(self:WBP Purchase Confirmation Dialog Object Reference=@n39, Delegate=@n42) {is=title K2Node_AddDelegate}
   > TryPushWidgetAsModalPanel!(execute, PushWidget:User Widget Object Reference=@n39) -> (then)
   n43 = SetAllInteractionsEnabled!() -> (OutputDelegate, then, ShouldEnable:bool=~) {is=title K2Node_CustomEvent}
   > SetIsInteractionEnabled!(
    execute
    self:Common Button Base Object Reference=[Get Button_GetVBucks(), Get Button_Purchase(), Get Button_UndoPurchase(), Get Button_Gift(), Get Button_ViewBundle()]
    bInIsInteractionEnabled:bool=@n43.ShouldEnable
   ) -> (then)
   UndoPurchase() {is=title K2Node_CustomEvent}
   > n44 = SetAllInteractionsEnabled(ShouldEnable=false) {is=title K2Node_CallFunction}
   n45 = OnPurchaseUndone!() -> (OutputDelegate, then, bSuccess:bool=~, RefundedItemTemplateIds:Array of Strings, PurchaseId:String, MtxReturned:int) {is=title K2Node_CustomEvent}
   UndoPurchase!(
    execute=@n44.then
    self:Fort Item Shop Offer VM Object Reference=Get FortItemShopOfferVM()
    OnUndoPurchaseCompleteCallback:"Delegate (by ref)"=@n45
   ) -> (then) {is=title K2Node_CallFunction}
   n46 = SetCanPurchase!() -> (OutputDelegate, then, CanPurchase:bool=~) {is=title K2Node_CustomEvent}
   > n47 = Branch(Condition=@n46.CanPurchase)
   > n48 = Get!(execute:Exec) -> (then:Exec, else:Exec, FortItemShopOfferVM:Fort Item Shop Offer VM Object Reference)
   n49 = CanPurchaseOffer!(self:Fort Item Shop Offer VM Object Reference=@n48) -> (ReturnValue:EOfferPurchaseError Enum=NoError)
   > Set Visibility(
    Visibility=Select(Option 0=Collapsed, Option 1=SelfHitTestInvisible, Index="Equal (Enum)"(A:EOfferPurchaseError Enum=@n49, B:EOfferPurchaseError Enum=NoError))
    self=Get Button_Purchase()
   )
   n50 = "Equal (Enum)"(A=@n49, B=CannotAffordItem)
   n51 = "Equal (Enum)"(A:EFortPriceCurrency Enum=GetCurrencyType(self=Get FortItemShopOfferVM()) -> (ReturnValue=Mtx), B:EFortPriceCurrency Enum=Mtx)
   > Set Visibility(Visibility=Select(Option 0=Collapsed, Option 1=SelfHitTestInvisible, Index=AND Boolean(A=@n51, B=@n50)), self=Get Button_GetVBucks())
   > SetPurchaseButtonStyle!(
    execute
    Purchase Button:Fort CTAButton Object Reference=Get Button_Purchase()
    GetVbucks Button:Fort CTAButton Object Reference=Get Button_GetVBucks()
   ) -> (then)
   Set Visibility(execute=@n47.else, Visibility=Collapsed, self=Get Button_Purchase())
   > Set Visibility(Visibility=Collapsed, self=Get Button_GetVBucks())
   n52 = ShowPurchaseOverlay!() -> (OutputDelegate, then, PurchasedItemTemplateIDs:Array of Strings, ShowAsRefund:bool=~) {is=title K2Node_CustomEvent}
   > n53 = Create Widget!(execute, Class=/ItemShop/Athena/AthenaItemShopOfferPurchasedOverlay.AthenaItemShopOfferPurchasedOverlay_C, OwningPlayer=GetOwningPlayer()) -> (then, ReturnValue:Athena Item Shop Offer Purchased Overlay Object Reference)
   > Bind Event to On Widget Deactivated(self:Common Activatable Widget Object Reference=@n53, Delegate=Create Event()) {is=title K2Node_AddDelegate}
   > Setup!(
    execute
    self:Athena Item Shop Purchase Overlay Modal Object Reference=@n53
    InQuantityPurchased:int=1
    InPurchasedItemTemplateIds:Array of Strings=@n52.PurchasedItemTemplateIDs
    bInShowAsRefund:bool=@n52.ShowAsRefund
   ) -> (then)
   > SetOffer!(
    execute
    self:Athena Item Shop Offer Details Widget Base Object Reference=@n53
    InOfferData:Fort Store Front Offer Info Object Reference=Get OfferInfo!(self:Fort Item Shop Offer VM Object Reference=Get FortItemShopOfferVM()) -> (OfferInfo:Athena Item Shop Offer Info Object Reference)
   ) -> (then)
   > TryPushWidgetAsModalPanel(PushWidget=@n53)
   n54 = SetCanViewBundle!() -> (OutputDelegate, then, Can View Bundle:bool=~) {is=title K2Node_CustomEvent}
   > Set Visibility(
    Visibility=Select(Option 0=Collapsed, Option 1=SelfHitTestInvisible, Index=@n54.Can View Bundle)
    self=Get VerticalBox_ViewBundle() -> (VerticalBox_ViewBundle:Vertical Box Object Reference)
   )
   > n55 = Get!(execute) -> (then, else, BundleOfferVM:Fort Item Shop Offer VM Object Reference)
   > Set Text(
    Text=Format Text!(
     Format="LOCTEXT(\"Part of the <text color=\\\"FFFFFF\\\" fontface=\\\"75 medium\\\">{BundleName}</>\")"
     BundleName:Text=GetTitle!(self:Fort Base Offer VM Object Reference=@n55) -> (ReturnValue:Text)
    ) -> (Result)
    self:Rich Text Block Object Reference=Get RichText_ParentOffer() -> (RichText_ParentOffer:Common Rich Text Block Object Reference)
   )
   AttemptPurchase() {is=title K2Node_CustomEvent}
   n56 = Get FortItemShopOfferVM()
   n57 = GetPurchaseConfirmationMessage_MatchmakingBan!(self:Fort Item Shop Offer VM Object Reference=@n56) -> (ReturnValue:Text)
   n58 = GetPurchaseConfirmationMessage_Generic!(self:Fort Item Shop Offer VM Object Reference=@n56) -> (ReturnValue:Text)
   > n59 = Branch(Condition=NAND Boolean(A=TextIsEmpty(InText=@n58), B=TextIsEmpty(InText=@n57)) {is=title K2Node_CommutativeAssociativeBinaryOperator})
   > ShowPurchaseConfirmation() {is=title K2Node_CallFunction}
   n60 = Get FortItemShopOfferVM()
   n61 = Branch(execute=@n59.else, Condition="Equal (Enum)"(A=GetCurrencyType(self=@n60) -> (ReturnValue=Mtx), B=RealMoney))
   n62 = Get FortItemShopOfferVM()
   > n63 = Branch(Condition=AND Boolean(
     A=HasMultiplePrices!(self:Fort Item Shop Offer VM Object Reference=@n62) -> (ReturnValue:bool)
     B=IsPurchaseChoiceRequestedEventBound!(self) -> (ReturnValue:bool)
    ))
   > BroadcastPurchaseChoiceRequestedEvent!(
    execute
    self:Purchase Choice Requester Interface Interface
    Requester:"Purchase Choice Requester Interface Interface (by ref)"=Self-Reference()
   ) -> (then)
   PurchaseOffer!(execute=[@n63.else, @n61.else], self, PriceIndex:int=0) -> (then)
   n64 = PurchaseOffer!() -> (OutputDelegate, then, PriceIndex:int) {is=title K2Node_CustomEvent}
   > PurchaseOffer!(
    execute
    self:Fort Item Shop Offer VM Object Reference=Get FortItemShopOfferVM()
    PriceIndex=@n64.PriceIndex
    OnPurchaseCompleteCallback:"Delegate (by ref)"=@n29
   ) -> (then, ReturnValue:EOfferPurchaseError Enum=NoError) {is=title K2Node_CallFunction}
   ShowCosmeticGateConfirmation() {is=title K2Node_CustomEvent}
   > n65 = Create Widget!(
    execute
    Class=/FortniteCoreUI/Frontend/CosmeticGating/WBP_CosmeticGate_ShopModal.WBP_CosmeticGate_ShopModal_C
    OwningPlayer
    FortItemShopOfferVM:Fort Item Shop Offer VM Object Reference=Get FortItemShopOfferVM()
   ) -> (then, ReturnValue:WBP Cosmetic Gate Shop Modal Object Reference)
   > Bind Event to On Dismissed(self:WBP Cosmetic Gate Shop Modal Object Reference=@n65, Delegate=Create Event()) {is=title K2Node_AddDelegate}
   > Bind Event to On Purchase Clicked(self:WBP Cosmetic Gate Shop Modal Object Reference=@n65, Delegate=Create Event()) {is=title K2Node_AddDelegate}
   > TryPushWidgetAsModalPanel(PushWidget=@n65)
   n66 = Event StartPurchase!() -> (OutputDelegate, then, PriceIndex:int=0) {is=title K2Node_Event}
   > PurchaseOffer!(execute, self:Self Object Reference, PriceIndex=@n66.PriceIndex) -> (then)
   n67 = Event PreConstruct()
   > n68 = Branch(Condition=@n67.IsDesignTime)
   SetMobileStyle(execute=[@n68.else, @n68.then]) {is=title K2Node_CallFunction}
   n69 = Get Preview_IsOwned() -> (Preview_IsOwned:bool)
   n70 = Get Preview_CanUndoPurchase() -> (Preview_CanUndoPurchase:bool)
   > SetCanPurchase!(execute, self, CanPurchase:bool=AND Boolean(A=NOT Boolean(A=@n69), B=NOT Boolean(A=@n70))) -> (then) {is=title K2Node_CallFunction}
   > SetCanUndoPurchase!(execute, self, CanUndo:bool=Get Preview_CanUndoPurchase()) -> (then) {is=title K2Node_CallFunction}
   > SetCanGift!(execute, self, IsGiftable:bool=Get Preview_CanGift() -> (Preview_CanGift:bool), IsOwned:bool=Get Preview_IsOwned()) -> (then) {is=title K2Node_CallFunction}
   > SetCanViewBundle!(execute, self, Can View Bundle:bool=Get Preview_CanViewBundle() -> (Preview_CanViewBundle:bool)) -> (then) {is=title K2Node_CallFunction}
   > SetIsGated!(
    execute
    self
    IsGated:bool=Get Preview_IsGated() -> (Preview_IsGated:bool)
    DisclaimerText:Text="INVTEXT(\"[PH] This Jam Track contains lyrics that may not be suitable in all experiences. Only the instrumental loops will be available when jamming in islands rated 10 years or lower, not the lyrics.\")"
   ) -> (then) {is=title K2Node_CallFunction}
   > SetRatingBoardWarning!(
    execute
    self
    ShowWarning:bool=Get Preview_ShowRatingBoardWarning() -> (Preview_ShowRatingBoardWarning:bool)
    WarningText:Text="INVTEXT(\"[PH] MUSIC DOWNLOADS NOT RATED BY PEGI - www.pegi.info\")"
   ) -> (then) {is=title K2Node_CallFunction}
   Event CancelPurchase() {is=title K2Node_Event}
   > EnableAllInteractions() {is=title K2Node_CallFunction}
   SetAllInteractionsEnabled(execute=@n42.then, ShouldEnable=true)
   PurchaseOffer(execute=@n40.then, PriceIndex=0)
   ShowPurchaseOverlay(execute=@n71.then, ShowAsRefund=true) {is=title K2Node_CallFunction}
   n71 = UpdateGrantedItemsIsOwned(execute=@n72.then, self=Get FortItemShopOfferVM())
   SetAllInteractionsEnabled(execute=@n45.then, ShouldEnable=true)
   > n72 = Branch(Condition=@n45.bSuccess)
  ]}
 GetDesiredFocusTarget {nodes [
   GetDesiredFocusTarget() {is=title K2Node_FunctionEntry}
   n1 = Get Button_Purchase()
   n2 = Get Button_UndoPurchase()
   n3 = IsVisible(self=@n2)
   n4 = Get Button_GetVBucks()
   n5 = Get Button_Gift()
   n6 = Get Button_ViewBundle()
   n7 = IsVisible(self=@n6)
   n8 = Select(
    Option 0:Widget Object Reference=Select(Option 0=Select(Option 1=@n6, Index=@n7), Option 1=@n5, Index=IsVisible(self=@n5))
    Option 1:Widget Object Reference=@n4
    Index=IsVisible(self=@n4)
   ) -> (ReturnValue:Widget Object Reference)
   > Return Node!(
    execute
    ReturnValue:Widget Object Reference=Select(Option 0=Select(Option 0=@n8, Option 1=@n2, Index=@n3), Option 1=@n1, Index=IsVisible(self=@n1))
   )
  ]}
 GetItemTemplateIDs {nodes [
   n1 = GetItemTemplateIDs!() -> (then, ItemInfo:Array of Purchased Item Info Structures) {is=title K2Node_FunctionEntry}
   > n2 = For Each Loop(Array:Array of Purchased Item Info Structures=@n1) -> (Array Element:Purchased Item Info Structure)
   n3 = Break Purchased Item Info!(PurchasedItemInfo:"Purchased Item Info Structure (by ref)"=@n2) -> (Item:Fort Item Object Reference, Quantity:int) {is=title K2Node_BreakStruct}
   n4 = Get TemplateIDs() -> (TemplateIDs:Array of Strings)
   > Add(
    TargetArray=@n4
    NewItem=GetTemplateId!(self:Mcp Item Definition Base Object Reference=GetItemDefinition!(self:Fort Item Object Reference=@n3) -> (ReturnValue:Fort Item Definition Object Reference)) -> (ReturnValue:String)
   ) {is=title K2Node_CallArrayFunction}
   Return Node!(execute=@n2.Completed, ItemTemplateIDs:Array of Strings=@n4)
  ]}
 GetParentOfferVM {nodes [
   n1 = GetParentOfferVM!() -> (then, ShopVM:Fort Item Shop VM Object Reference, ChildOfferVM:Fort Item Shop Offer VM Object Reference, CanUndoPurchase:bool=~) {is=title K2Node_FunctionEntry}
   > n2 = Is Valid(InputObject=@n1)
   Return Node!(execute=@n2.Is Not Valid, ParentOfferVM:Fort Item Shop Offer VM Object Reference)
   > n3 = FindParentOffer!(execute, self:Fort Item Shop VM Object Reference=@n1, ChildOffer:Fort Item Shop Offer VM Object Reference=@n1.ChildOfferVM) -> (then, ReturnValue:Fort Item Shop Offer VM Object Reference)
   > n4 = Branch(Condition=AND Boolean(A=IsValid(Object=@n3), B=NOT Boolean(A=@n1.CanUndoPurchase)))
   > Return Node(ParentOfferVM=@n3)
   Return Node(execute=@n4.else)
  ]}
 GetPurchaseChoiceOfferInfo {nodes [
   GetPurchaseChoiceOfferInfo() {is=title K2Node_FunctionEntry}
   > Return Node!(execute, ReturnValue:Fort Store Front Offer Info Object Reference)
  ]}
 GetPurchaseChoiceOwningPlayer {nodes [
   GetPurchaseChoiceOwningPlayer() {is=title K2Node_FunctionEntry}
   > Return Node(ReturnValue:Fort Player Controller Object Reference)
  ]}
 IsPurchaseChoiceRequestedEventBound {nodes [
   IsPurchaseChoiceRequestedEventBound() {is=title K2Node_FunctionEntry}
   > Return Node(ReturnValue:bool=false)
  ]}
 NavUpFromViewBundle {nodes [
   NavUpFromViewBundle!() -> (then, Navigation:EUINavigation Enum) {is=title K2Node_FunctionEntry}
   > Return Node(ReturnValue:Widget Object Reference=GetDesiredFocusTarget!() -> (ReturnValue:Widget Object Reference) {is=title K2Node_CallFunction})
  ]}
 OnDetailsButtonsHoveredChanged {nodes [OnDetailsButtonsHoveredChanged!() -> (then, IsHovered:bool=~) {is=title K2Node_FunctionEntry}]}
 SetMobileStyle {nodes [
   SetMobileStyle() {is=title K2Node_FunctionEntry}
   > Branch(Condition=IsUsingCompactUI())
   n1 = SlotAsOverlaySlot(Widget=Get ItemDescription() -> (ItemDescription:WBP UIKit Item Description Base Object Reference))
   > SetHorizontalAlignment(self=@n1, InHorizontalAlignment=HAlign_Fill)
   > SetPadding!(
    execute
    self:Overlay Slot Object Reference=@n1
    InPadding_Left:float=80
    InPadding_Top:float=0
    InPadding_Right:float=80
    InPadding_Bottom:float=0
   ) -> (then)
   n2 = Get Text_OutDate() -> (Text_OutDate:Common Text Object Reference)
   > SetAutoWrapText!(execute, self:Text Object Reference=@n2, InAutoTextWrap:bool=true) -> (then)
   > SetHorizontalAlignment(self:Vertical Box Slot Object Reference=SlotAsVerticalBoxSlot(Widget=@n2), InHorizontalAlignment=HAlign_Fill)
   > SetRequiresHold!(
    execute
    self:Common Button Base Object Reference=Get Button_Purchase()
    bInRequiresHold:bool=NOT Boolean(A=GetConfigDataBoolWithDefault!(EntryName:String=shop_confirm_purchase_through_dialog, bDefaultValue:bool=false) -> (ReturnValue:bool))
   ) -> (then)
   n3 = GetViewportSize!() -> (ReturnValue_X:double, ReturnValue_Y:double)
   > n4 = Branch(Condition="float >= float"(A=float / float(A=@n3, B=@n3.ReturnValue_Y), B=Get Mobile Viewport Ratio() -> (Mobile Viewport Ratio:double)))
   n5 = SlotAsVerticalBoxSlot(Widget=Get Warning_RatingBoard())
   > SetPadding(self:Vertical Box Slot Object Reference=@n5, InPadding_Left=704, InPadding_Top=16, InPadding_Right=0, InPadding_Bottom=0)
   > SetHorizontalAlignment(self=@n5, InHorizontalAlignment=HAlign_Fill)
   n6 = SlotAsVerticalBoxSlot(Widget=Get Warning_Cosmetic())
   > SetPadding(self=@n6, InPadding_Left=704, InPadding_Top=16, InPadding_Right=0, InPadding_Bottom=0)
   > SetHorizontalAlignment(self=@n6, InHorizontalAlignment=HAlign_Fill)
   > SetVerticalAlignment(
    self:Horizontal Box Slot Object Reference=[
     SlotAsHorizontalBoxSlot!(Widget:Widget Object Reference=Get Button_Gift()) -> (ReturnValue:Horizontal Box Slot Object Reference)
     SlotAsHorizontalBoxSlot(Widget=Get Button_Purchase())
     SlotAsHorizontalBoxSlot(Widget=Get Button_GetVBucks())
     SlotAsHorizontalBoxSlot(Widget=Get Button_UndoPurchase())
    ]
    InVerticalAlignment=VAlign_Center
   )
   n7 = Get VerticalBox_ViewBundle()
   > n8 = AddChildToHorizontalBox!(
    execute
    self:Horizontal Box Object Reference=Get HorizontalBox_Buttons() -> (HorizontalBox_Buttons:Horizontal Box Object Reference)
    Content:Widget Object Reference=@n7
   ) -> (then, ReturnValue:Horizontal Box Slot Object Reference)
   > SetPadding(self:Horizontal Box Slot Object Reference=@n8, InPadding_Left=25, InPadding_Top=0, InPadding_Right=0, InPadding_Bottom=0)
   > SetRenderTranslation(self=@n7, Translation=(X=0,Y=-18))
   SetPadding(execute=@n4.else, self:Vertical Box Slot Object Reference=@n5, InPadding_Left=400, InPadding_Top=0, InPadding_Right=0, InPadding_Bottom=0)
   > SetHorizontalAlignment(self=@n5, InHorizontalAlignment=HAlign_Fill)
   > SetPadding(self=@n6, InPadding_Left=400, InPadding_Top=0, InPadding_Right=0, InPadding_Bottom=0)
   > SetHorizontalAlignment(self=@n6, InHorizontalAlignment=HAlign_Fill)
  ]}
 ShouldShowOfferPurchasedOverlay {nodes [
   ShouldShowOfferPurchasedOverlay() {is=title K2Node_FunctionEntry}
   n1 = Get bUseItemPresentationScreenOnItemPurchased!(self:Fort Runtime Options Object Reference=GetRuntimeOptions()) -> (bUseItemPresentationScreenOnItemPurchased:bool)
   n2 = Get OfferInfo(self=Get FortItemShopOfferVM())
   n3 = IsBattlePassLevelBundle!(self:Fort Store Front Offer Info Object Reference=@n2) -> (ReturnValue:bool)
   n4 = GetPriceCurrencyType!(self:Fort Store Front Offer Info Object Reference=@n2, PriceIndex:int=0) -> (ReturnValue:EStoreCurrencyType Enum=RealMoney)
   > Return Node(ReturnValue:bool=AND Boolean!(A="Equal (Enum)"(A:EStoreCurrencyType Enum=@n4, B:EStoreCurrencyType Enum=MtxCurrency), B=NOT Boolean(A=@n3), C:bool=NOT Boolean(A=@n1)) -> (ReturnValue))
  ]}
 VM_CanUndoPurchaseUpdated {nodes [
   n1 = VM_CanUndoPurchaseUpdated!() -> (then, CanUndoPurchase:bool=~) {is=title K2Node_FunctionEntry}
   > SetCanUndoPurchase(CanUndo=@n1)
  ]}
 VM_GrantedOffersUpdated {nodes [
   n1 = VM_GrantedOffersUpdated!() -> (then, GrantedItems:Array of Fort Item VM Object References) {is=title K2Node_FunctionEntry}
   > n2 = For Each Loop(Array:Array of Fort Item VM Object References=@n1) -> (Array Element:Fort Item VM Object Reference)
   > n3 = Cast To AthenaItemVM!(execute, Object=@n2) -> (then, CastFailed, AsAthena Item VM:Athena Item VM Object Reference) {is=title K2Node_DynamicCast}
   > n4 = Cast To AthenaCosmeticItemDefinition!(execute, Object=Get ItemDefinition!(self:Fort Item VM Object Reference=@n3) -> (ItemDefinition:Fort Item Definition Object Reference)) -> (then, CastFailed, AsAthena Cosmetic Item Definition:Athena Cosmetic Item Definition Object Reference) {is=title K2Node_DynamicCast}
   > Add(
    TargetArray:Array of Athena Cosmetic Item Definition Object References=Get ItemDefs() -> (ItemDefs:Array of Athena Cosmetic Item Definition Object References)
    NewItem:"Athena Cosmetic Item Definition Object Reference (by ref)"=@n4
   )
   n5 = ConvertCosmeticItemDefinitionsToInterfaces!(execute=@n2.Completed, CosmeticItemDefinitions:Array of Athena Cosmetic Item Definition Object References=Get ItemDefs()) -> (then, OutCosmeticItemGatingInterfaces:Array of Cosmetic Item Gating Interface Interfaces)
   > n6 = GetRatingRestrictionTextForItems!(
    execute
    LocalPlayer:Local Player Object Reference=GetOwningLocalPlayer!(self:Widget Object Reference) -> (ReturnValue:Local Player Object Reference)
    CosmeticItems:Array of Cosmetic Item Gating Interface Interfaces=@n5
   ) -> (then, ReturnValue:Text)
   > SetIsGated(IsGated=NOT Boolean(A=TextIsEmpty(InText=@n6)), DisclaimerText=@n6)
  ]}
 VM_IsGiftableUpdated {nodes [
   n1 = VM_IsGiftableUpdated!() -> (then, IsGiftable:bool=~) {is=title K2Node_FunctionEntry}
   n2 = IsGiftingDisabled!(self:Fort Runtime Options Object Reference=GetRuntimeOptions()) -> (ReturnValue:bool)
   > SetCanGift(
    IsGiftable=AND Boolean!(A=@n1, B=NOT Boolean(A=@n2)) -> (ReturnValue)
    IsOwned=Get bOwned!(self:Fort Item Shop Offer VM Object Reference=Get FortItemShopOfferVM()) -> (bOwned:bool)
   )
  ]}
 VM_IsOwnedUpdated {nodes [
   n1 = VM_IsOwnedUpdated!() -> (then, Owned:bool=~) {is=title K2Node_FunctionEntry}
   > SetCanPurchase(CanPurchase=NOT Boolean(A=@n1))
   n2 = IsGiftingDisabled(self=GetRuntimeOptions())
   n3 = Get OfferInfo(self=Get FortItemShopOfferVM())
   > SetCanGift(
    IsGiftable=AND Boolean(A=IsGiftable!(self:Fort Store Front Offer Info Object Reference=@n3) -> (ReturnValue:bool), B=NOT Boolean(A=@n2))
    IsOwned=Get Owned() -> (Owned:bool)
   )
  ]}
 VM_ParentOfferUpdated {nodes [
   n1 = VM_ParentOfferUpdated!() -> (then, ParentOfferVM:Fort Item Shop Offer VM Object Reference) {is=title K2Node_FunctionEntry}
   > n2 = Set BundleOfferVM(BundleOfferVM:Fort Item Shop Offer VM Object Reference=@n1) -> (Output_Get:Fort Item Shop Offer VM Object Reference)
   > SetCanViewBundle(Can View Bundle=IsValid(Object=@n2))
  ]}
 VM_RatingBoardWarningUpdated {nodes [
   n1 = VM_RatingBoardWarningUpdated!() -> (then, RatingBoardWarning:Text) {is=title K2Node_FunctionEntry}
   > SetRatingBoardWarning(ShowWarning=NOT Boolean(A=TextIsEmpty(InText=@n1)), WarningText=@n1)
  ]}
]
variables [
 BundleOfferVM {type=FortItemShopOfferVM, default=null}
 IsOfferGated {type=boolean}
 Mobile Viewport Ratio {type=number, default=1.34}
 OnDetailsButtonsHoveredChanged {}
 Preview_CanGift {type=boolean, default=true}
 Preview_CanUndoPurchase {type=boolean}
 Preview_CanViewBundle {type=boolean, default=true}
 Preview_IsGated {type=boolean, default=true}
 Preview_IsOwned {type=boolean}
 Preview_ShowRatingBoardWarning {type=boolean, default=true}
]
widgets {
 info {inheritedWidgetCount=0, parentClass=&UserWidget, rootWidgetClass=&Overlay, widgetCount=19}
 widgets [
  Overlay_0 {widgetClassPath=&Overlay, widget=Overlay_0}
  ItemDescription {
   widgetClassPath=/FortUILibrary/Systems/Items/WBP_UIKit_ItemDescription_Base
   bIsVariable=true
   parent=Overlay_0
   widget=ItemDescription
   properties {
    hideBodyText=true
    hideCategoryText=true
    hideRarityTag=true
    hideStyleText=true
    horizontalAlignment=HAlign_Center
    bodyTextPadding=<ltrb 0 0 0 0>
    compatibilityPadding=<ltrb 0 16 0 0>
    titleFont {
     bForceMonospaced=false
     bMaterialIsStencil=false
     fontMaterial=null
     fontObject=&HeadingNow
     letterSpacing=0
     monospacedWidth=1
     size=25.5
     skewAmount=0
     typefaceFontName=86 Bold Italic
     outlineSettings {
      bApplyOutlineToDropShadows=false
      bMiteredCorners=false
      bSeparateFillAlpha=false
      outlineMaterial=null
      outlineSize=0
      outlineColor=<rgba 0 0 0 1>
     }
    }
   }
   slot {class=OverlaySlot, properties {horizontalAlignment=HAlign_Center}}
  }
  VerticalBox_0 {
   widgetClassPath=&VerticalBox
   parent=Overlay_0
   widget=VerticalBox_0
   slot {class=OverlaySlot, properties {horizontalAlignment=HAlign_Fill, verticalAlignment=VAlign_Bottom}}
  }
  Warning_RatingBoard {
   widgetClassPath=/FortUILibrary/UIKit/InfoBox/WBP_UIKit_InfoBox
   bIsVariable=true
   parent=VerticalBox_0
   widget=Warning_RatingBoard
   properties {previewText="[PH] MUSIC DOWNLOADS NOT RATED BY PEGI - www.pegi.info", showIcon=false, visibility=Collapsed}
   slot {class=VerticalBoxSlot, properties {horizontalAlignment=HAlign_Center}}
  }
  Warning_Cosmetic {
   widgetClassPath=&WBP_UIKit_InfoBox
   bIsVariable=true
   parent=VerticalBox_0
   widget=Warning_Cosmetic
   properties {previewText="This Jam Track contains lyrics that may not be suitable in all experiences. Only the instrumental loops will be available when jamming in islands rated 10 years or lower, not the lyrics."}
   slot {class=VerticalBoxSlot, properties {horizontalAlignment=HAlign_Center, padding=<ltrb 0 16 0 0>}}
  }
  WBP_ItemShop_ListView_Price {
   widgetClassPath=&/WBP_ItemShop_ListView_Price
   bIsVariable=true
   parent=VerticalBox_0
   widget=WBP_ItemShop_ListView_Price
   properties {realPriceLegalText_Optional=&self:WidgetTree.Text_RealPriceLegal}
   slot {class=VerticalBoxSlot, properties {horizontalAlignment=HAlign_Center, padding=<ltrb 0 24 0 24>}}
  }
  HorizontalBox_Buttons {
   widgetClassPath=&HorizontalBox
   bIsVariable=true
   parent=VerticalBox_0
   widget=HorizontalBox_Buttons
   slot {class=VerticalBoxSlot, properties {padding=<ltrb 90 0 90 16>}}
  }
  Button_Purchase {
   widgetClassPath=/FortUILibrary/UIKit/Buttons/WBP_UIKit_Button_Loud
   bIsVariable=true
   parent=HorizontalBox_Buttons
   widget=Button_Purchase
   properties {
    bRequiresHold=true
    holdData=/FortUILibrary/Input/PurchaseInteraction_HoldData.PurchaseInteraction_HoldData_C
    minHeight=64
    minWidth=300
    mobileMinHeight=80
    navigation=&self:WidgetTree.Button_Purchase.WidgetNavigation_0
    overrideDefaultHeight=true
    text=Purchase
   }
   slot {class=HorizontalBoxSlot, properties {padding=<ltrb 0 0 24 0>, size {sizeRule=Fill, value=0.65}}}
  }
  Button_GetVBucks {
   widgetClassPath=&WBP_UIKit_Button_Loud
   bIsVariable=true
   parent=HorizontalBox_Buttons
   widget=Button_GetVBucks
   properties {
    bShouldUseFallbackDefaultInputAction=false
    minHeight=64
    minWidth=300
    mobileMinHeight=80
    navigation=&self:WidgetTree.Button_GetVBucks.WidgetNavigation_0
    overrideDefaultHeight=true
    text=Get V-Bucks
   }
   slot {class=HorizontalBoxSlot, properties {padding=<ltrb 0 0 24 0>, size {sizeRule=Fill, value=0.65}}}
  }
  Button_UndoPurchase {
   widgetClassPath=&WBP_UIKit_Button_Regular
   bIsVariable=true
   parent=HorizontalBox_Buttons
   widget=Button_UndoPurchase
   properties {
    bRequiresHold=true
    holdData=&PurchaseInteraction_HoldData_C
    minHeight=56
    minWidth=300
    navigation=&self:WidgetTree.Button_UndoPurchase.WidgetNavigation_0
    overrideDefaultHeight=true
    text=Cancel purchase
   }
   slot {class=HorizontalBoxSlot, properties {padding=<ltrb 0 0 24 0>, size {sizeRule=Fill, value=0.65}}}
  }
  Button_Gift {
   widgetClassPath=&WBP_UIKit_Button_Regular
   bIsVariable=true
   parent=HorizontalBox_Buttons
   widget=Button_Gift
   properties {
    bShouldUseFallbackDefaultInputAction=false
    icon=/FortUILibrary/Textures/Icons/T_UI_Icon_UIKit_Gift
    iconMaterial=/FortUILibrary/UIKit/Buttons/MaterialInstances/CTAButtons/MI_UIKit_IconQuiet
    minHeight=56
    minWidth=300
    navigation=&self:WidgetTree.Button_Gift.WidgetNavigation_0
    overrideDefaultHeight=true
    text=""
   }
   slot {class=HorizontalBoxSlot, properties {size {sizeRule=Automatic, value=0.35}}}
  }
  VerticalBox_ViewBundle {
   widgetClassPath=&VerticalBox
   bIsVariable=true
   parent=VerticalBox_0
   widget=VerticalBox_ViewBundle
   slot {class=VerticalBoxSlot, properties {padding=<ltrb 90 0 90 12>}}
  }
  RichText_ParentOffer {
   widgetClassPath=&CommonRichTextBlock
   bIsVariable=true
   parent=VerticalBox_ViewBundle
   widget=RichText_ParentOffer
   properties {
    bOverrideDefaultStyle=true
    justification=Center
    text="[PH] Part of the <text color=\"FFFFFF\" fontface=\"75 medium\">{BundleName}</>"
    wrapTextAt=640
    defaultTextStyleOverride {
     overflowPolicy=Clip
     transformPolicy=None
     colorAndOpacity {colorUseRule=UseColor_Specified, specifiedColor=<rgba 1 1 1 1>}
     font {
      bForceMonospaced=false
      bMaterialIsStencil=false
      fontMaterial=null
      fontObject=&HeadingNow
      letterSpacing=0
      monospacedWidth=1
      size=15.75
      skewAmount=0
      typefaceFontName=74 Regular
      outlineSettings {
       bApplyOutlineToDropShadows=false
       bMiteredCorners=false
       bSeparateFillAlpha=false
       outlineMaterial=null
       outlineSize=0
       outlineColor=<rgba 0 0 0 1>
      }
     }
     highlightColor {colorUseRule=UseColor_Specified, specifiedColor=<rgba 0 0 0 1>}
     highlightShape {
      bIsDynamicallyLoaded=false
      drawAs=Image
      imageType=NoImage
      mirroring=NoMirror
      resourceName=None
      resourceObject=null
      tiling=NoTile
      imageSize=<32 32>
      margin=<ltrb 0 0 0 0>
      outlineSettings {
       bUseBrushTransparency=false
       roundingType=HalfHeightRadius
       width=0
       color {colorUseRule=UseColor_Specified, specifiedColor=<rgba 0 0 0 0>}
       cornerRadii=<0 0 0 0>
      }
      tintColor {colorUseRule=UseColor_Specified, specifiedColor=<rgba 1 1 1 1>}
      uVRegion {bIsValid=false, max=<0 0>, min=<0 0>}
     }
     selectedBackgroundColor {colorUseRule=UseColor_Foreground, specifiedColor=<rgba 1 0 1 1>}
     shadowColorAndOpacity=<rgba 0 0 0 1>
     shadowOffset=<0 0>
     strikeBrush {
      bIsDynamicallyLoaded=false
      drawAs=Image
      imageType=NoImage
      mirroring=NoMirror
      resourceName=None
      resourceObject=null
      tiling=NoTile
      imageSize=<32 32>
      margin=<ltrb 0 0 0 0>
      outlineSettings {
       bUseBrushTransparency=false
       roundingType=HalfHeightRadius
       width=0
       color {colorUseRule=UseColor_Specified, specifiedColor=<rgba 0 0 0 0>}
       cornerRadii=<0 0 0 0>
      }
      tintColor {colorUseRule=UseColor_Specified, specifiedColor=<rgba 1 1 1 1>}
      uVRegion {bIsValid=false, max=<0 0>, min=<0 0>}
     }
     underlineBrush {
      bIsDynamicallyLoaded=false
      drawAs=Image
      imageType=NoImage
      mirroring=NoMirror
      resourceName=None
      resourceObject=null
      tiling=NoTile
      imageSize=<32 32>
      margin=<ltrb 0 0 0 0>
      outlineSettings {
       bUseBrushTransparency=false
       roundingType=HalfHeightRadius
       width=0
       color {colorUseRule=UseColor_Specified, specifiedColor=<rgba 0 0 0 0>}
       cornerRadii=<0 0 0 0>
      }
      tintColor {colorUseRule=UseColor_Specified, specifiedColor=<rgba 1 1 1 1>}
      uVRegion {bIsValid=false, max=<0 0>, min=<0 0>}
     }
    }
   }
   slot {class=VerticalBoxSlot, properties {padding=<ltrb 0 0 0 10>}}
  }
  Button_ViewBundle {
   widgetClassPath=&WBP_UIKit_Button_Regular
   bIsVariable=true
   parent=VerticalBox_ViewBundle
   widget=Button_ViewBundle
   properties {
    bShouldUseFallbackDefaultInputAction=false
    buttonSize=Small
    navigation=&self:WidgetTree.Button_ViewBundle.WidgetNavigation_1
    overrideDefaultHeight=true
    text=View Bundle
   }
   slot {class=VerticalBoxSlot}
  }
  Text_OutDate {
   widgetClassPath=&CommonTextBlock
   bIsVariable=true
   parent=VerticalBox_0
   widget=Text_OutDate
   properties {
    justification=Center
    lineHeightPercentage=1.15
    mobileFontSizeMultiplier=1.5
    text="[PH] This item will be removed from the Item Shop on 01/01/24 at 6:00 PM your local time (GMT-4)."
    colorAndOpacity {colorUseRule=UseColor_Specified, specifiedColor=<rgba 1 1 1 0.65>}
    font {
     bForceMonospaced=false
     bMaterialIsStencil=false
     fontMaterial=null
     fontObject=&HeadingNow
     letterSpacing=0
     monospacedWidth=1
     size=10.5
     skewAmount=0
     typefaceFontName=74 Regular
     outlineSettings {
      bApplyOutlineToDropShadows=false
      bMiteredCorners=false
      bSeparateFillAlpha=false
      outlineMaterial=null
      outlineSize=0
      outlineColor=<rgba 0 0 0 1>
     }
    }
    shadowOffset=<0 0>
   }
   slot {class=VerticalBoxSlot, properties {horizontalAlignment=HAlign_Center, padding=<ltrb 0 8 0 8>}}
  }
  Text_Legal_UndoPurchase {
   widgetClassPath=&CommonTextBlock
   bIsVariable=true
   parent=VerticalBox_0
   widget=Text_Legal_UndoPurchase
   properties {
    applyLineHeightToBottomLine=false
    autoWrapText=true
    justification=Center
    lineHeightPercentage=1.2
    text=Cancel Purchase is available until you equip or use your purchase during gameplay. Only items purchased within the last 24 hours are eligible for Cancel Purchase.
    colorAndOpacity {colorUseRule=UseColor_Specified, specifiedColor=<rgba 1 1 1 0.65>}
    font {
     bForceMonospaced=false
     bMaterialIsStencil=false
     fontMaterial=null
     fontObject=&HeadingNow
     letterSpacing=0
     monospacedWidth=1
     size=10.5
     skewAmount=0
     typefaceFontName=74 Regular
     outlineSettings {
      bApplyOutlineToDropShadows=false
      bMiteredCorners=false
      bSeparateFillAlpha=false
      outlineMaterial=null
      outlineSize=0
      outlineColor=<rgba 0 0 0 1>
     }
    }
    shadowOffset=<0 0>
   }
   slot {class=VerticalBoxSlot, properties {padding=<ltrb 0 0 0 8>}}
  }
  Text_RefundDisclaimer {
   widgetClassPath=&CommonTextBlock
   bIsVariable=true
   parent=VerticalBox_0
   widget=Text_RefundDisclaimer
   properties {
    applyLineHeightToBottomLine=false
    autoWrapText=true
    justification=Center
    lineHeightPercentage=1.2
    mobileFontSizeMultiplier=1.5
    text=This purchase is not eligible for refund.
    colorAndOpacity {colorUseRule=UseColor_Specified, specifiedColor=<rgba 1 1 1 0.65>}
    font {
     bForceMonospaced=false
     bMaterialIsStencil=false
     fontMaterial=null
     fontObject=&HeadingNow
     letterSpacing=0
     monospacedWidth=1
     size=10.5
     skewAmount=0
     typefaceFontName=74 Regular
     outlineSettings {
      bApplyOutlineToDropShadows=false
      bMiteredCorners=false
      bSeparateFillAlpha=false
      outlineMaterial=null
      outlineSize=0
      outlineColor=<rgba 0 0 0 1>
     }
    }
    shadowOffset=<0 0>
   }
   slot {class=VerticalBoxSlot, properties {padding=<ltrb 0 0 0 8>}}
  }
  Text_Legal {
   widgetClassPath=&CommonTextBlock
   parent=VerticalBox_0
   widget=Text_Legal
   properties {
    applyLineHeightToBottomLine=false
    autoWrapText=true
    bAutoCollapseWithEmptyText=true
    justification=Center
    lineHeightPercentage=1.2
    mobileFontSizeMultiplier=1.5
    text="[PH] These cosmetic items grant no competitive advantage. Outfits and wraps don't include weapons. Not all cosmetics are usable in all experiences. For more details on cosmetic variations and limitations of use, see fn.gg/cosmetics"
    colorAndOpacity {colorUseRule=UseColor_Specified, specifiedColor=<rgba 1 1 1 0.65>}
    font {
     bForceMonospaced=false
     bMaterialIsStencil=false
     fontMaterial=null
     fontObject=&HeadingNow
     letterSpacing=0
     monospacedWidth=1
     size=10.5
     skewAmount=0
     typefaceFontName=74 Regular
     outlineSettings {
      bApplyOutlineToDropShadows=false
      bMiteredCorners=false
      bSeparateFillAlpha=false
      outlineMaterial=null
      outlineSize=0
      outlineColor=<rgba 0 0 0 1>
     }
    }
    shadowOffset=<0 0>
   }
   slot {class=VerticalBoxSlot, properties {padding=<ltrb 0 0 0 8>}}
  }
  Text_RealPriceLegal {
   widgetClassPath=&CommonTextBlock
   parent=VerticalBox_0
   widget=Text_RealPriceLegal
   properties {
    applyLineHeightToBottomLine=false
    autoWrapText=true
    bAutoCollapseWithEmptyText=true
    justification=Center
    lineHeightPercentage=1.2
    mobileFontSizeMultiplier=1.5
    text="*To buy exactly 1,000 V-Bucks, you’d pay €8.99."
    colorAndOpacity {colorUseRule=UseColor_Specified, specifiedColor=<rgba 1 1 1 0.65>}
    font {
     bForceMonospaced=false
     bMaterialIsStencil=false
     fontMaterial=null
     fontObject=&HeadingNow
     letterSpacing=0
     monospacedWidth=1
     size=10.5
     skewAmount=0
     typefaceFontName=74 Regular
     outlineSettings {
      bApplyOutlineToDropShadows=false
      bMiteredCorners=false
      bSeparateFillAlpha=false
      outlineMaterial=null
      outlineSize=0
      outlineColor=<rgba 0 0 0 1>
     }
    }
    shadowOffset=<0 0>
   }
   slot {class=VerticalBoxSlot}
  }
 ]
}
